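# Main multiplayer locker screen widget, hosting category tiles, side-navigation tabs, a presets panel and an item preview.
# Category clicks reverse the intro and push the configured editor or items screen (modal if configured), while Edit Styles pushes the variant selector.
# Presets expand via swipe, wheel, click or navigation for equippable groups, with a separate mobile animation.
# Widget animations drive camera and preview layout values, sliding the 3D camera as the active screen changes.
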WidgetBlueprint /Locker/MPL/FortMPLocker
animations [
 TogglePresetsFullscreen {
  binding_count=4
  length_seconds=0.3
  object=&self:TogglePresetsFullscreen
  sequence {bindings [
    GridSlot_0 {
     id=#1
     parent_id=#2
     possessed_class=GridSlot
     tracks [{
       type=MovieSceneMarginTrack
       display_name=Padding
       sections [{
         type=MovieSceneMarginSection
         end_frame=6
         start_frame=0
         channels [
          Left {type=MovieSceneScriptingFloatChannel}
          Top {type=MovieSceneScriptingFloatChannel}
          Right {type=MovieSceneScriptingFloatChannel}
          Bottom {type=MovieSceneScriptingFloatChannel, keys [[0,0],[5,0],[6,32]]}
         ]
        }]
      }]
    }
    PresetsWidget {
     id=#2
     possessed_class=WBP_MPLocker_Presets_C
     tracks [{
       type=MovieSceneFloatTrack
       display_name=Render Opacity
       sections [{type=MovieSceneFloatSection, end_frame=4, start_frame=0, channels [None {type=MovieSceneScriptingFloatChannel, keys [[0,0.3],[4.0043,1]]}]}]
      }]
    }
    FortMPLocker_C_82 {
     id=#3
     possessed_class=FortMPLocker_C
     tracks [
      {
       type=MovieSceneDoubleTrack
       display_name=Presets Menu Size
       sections [{
         type=MovieSceneDoubleSection
         end_frame=6
         start_frame=0
         channels [None {type=MovieSceneScriptingDoubleChannel, thinned_from=3, keys [[0,1],[6.0003,0]]}]
        }]
      }
      {
       type=MovieSceneByteTrack
       display_name=Visibility
       sections [{type=MovieSceneByteSection, end_frame=6, start_frame=0, channels [None {type=MovieSceneScriptingByteChannel, keys [[0,4],[1,3],[6.0003,4]]}]}]
      }
     ]
    }
    Categories {
     id=#4
     possessed_class=WBP_MPLocker_Categories_C
     tracks [{
       type=MovieSceneFloatTrack
       display_name=Render Opacity
       sections [{type=MovieSceneFloatSection, end_frame=3, start_frame=0, channels [None {type=MovieSceneScriptingFloatChannel, keys [[0.005,1],[3.0037,0]]}]}]
      }]
    }
   ]}
 }
 CameraSlideOut {
  binding_count=1
  length_seconds=0.5
  object=&self:CameraSlideOut
  sequence {bindings [FortMPLocker_C_13 {
     id=#5
     possessed_class=FortMPLocker_C
     tracks [
      {
       type=MovieSceneDoubleTrack
       display_name=Camera Position
       sections [{type=MovieSceneDoubleSection, end_frame=9, start_frame=-1, channels [None {type=MovieSceneScriptingDoubleChannel, keys [[1,0],[8,1]]}]}]
      }
      {
       type=MovieSceneByteTrack
       display_name=Visibility
       sections [{type=MovieSceneByteSection, end_frame=9, start_frame=-1, channels [None {type=MovieSceneScriptingByteChannel, keys [[0,3],[8.9997,4]]}]}]
      }
     ]
    }]}
 }
 CameraSlideIn {
  binding_count=1
  length_seconds=0.5
  object=&self:CameraSlideIn
  sequence {bindings [FortMPLocker_C_13 {
     id=#5
     possessed_class=FortMPLocker_C
     tracks [
      {
       type=MovieSceneDoubleTrack
       display_name=Camera Position
       sections [{type=MovieSceneDoubleSection, end_frame=9, start_frame=-1, channels [None {type=MovieSceneScriptingDoubleChannel, keys [[1,1],[8,0]]}]}]
      }
      {
       type=MovieSceneByteTrack
       display_name=Visibility
       sections [{type=MovieSceneByteSection, end_frame=9, start_frame=-1, channels [None {type=MovieSceneScriptingByteChannel, keys [[0,3],[8.9997,4]]}]}]
      }
     ]
    }]}
 }
 Intro {
  binding_count=3
  length_seconds=0.2501
  object=&self:Intro
  sequence {bindings [
    Categories {
     id=#6
     possessed_class=WBP_MPLocker_Categories_C
     tracks [
      {
       type=MovieSceneFloatTrack
       display_name=Render Opacity
       sections [{type=MovieSceneFloatSection, end_frame=5, start_frame=0, channels [None {type=MovieSceneScriptingFloatChannel, keys [[0,0],[4.9997,1]]}]}]
      }
      {
       type=MovieScene2DTransformTrack
       sections [{
         type=MovieScene2DTransformSection
         end_frame=5
         start_frame=0
         channels [Translation.X {type=MovieSceneScriptingFloatChannel, keys [[0,-200],[5,0]]}]
        }]
      }
     ]
    }
    CategoryTabs_NavWrapper {
     id=#7
     possessed_class=GridPanel
     tracks [
      {
       type=MovieSceneFloatTrack
       display_name=Render Opacity
       sections [{type=MovieSceneFloatSection, end_frame=5, start_frame=0, channels [None {type=MovieSceneScriptingFloatChannel, keys [[0,0],[5.0003,1]]}]}]
      }
      {
       type=MovieScene2DTransformTrack
       sections [{
         type=MovieScene2DTransformSection
         end_frame=5
         start_frame=0
         channels [Translation.X {type=MovieSceneScriptingFloatChannel, keys [[0,-175],[5,0]]}]
        }]
      }
     ]
    }
    PresetsWidget {
     id=#8
     possessed_class=WBP_MPLocker_Presets_C
     tracks [
      {
       type=MovieSceneFloatTrack
       display_name=Render Opacity
       sections [{type=MovieSceneFloatSection, end_frame=5, start_frame=0, channels [None {type=MovieSceneScriptingFloatChannel, keys [[0,0],[5,0.3]]}]}]
      }
      {
       type=MovieScene2DTransformTrack
       sections [{
         type=MovieScene2DTransformSection
         end_frame=5
         start_frame=0
         channels [Translation.Y {type=MovieSceneScriptingFloatChannel, keys [[0,150],[5,0]]}]
        }]
      }
     ]
    }
   ]}
 }
 PresetsPeek {
  binding_count=1
  length_seconds=0.1533
  object=&self:PresetsPeek
  sequence {bindings [Presets_NavWrapper {
     id=#9
     possessed_class=GridPanel
     tracks [{
       type=MovieScene2DTransformTrack
       sections [{
         type=MovieScene2DTransformSection
         end_frame=3
         start_frame=0
         channels [
          Translation.X {type=MovieSceneScriptingFloatChannel, const [0,2]}
          Translation.Y {type=MovieSceneScriptingFloatChannel, keys [[0,4],[3.0653,0]]}
          Angle {type=MovieSceneScriptingFloatChannel}
          Scale.X {type=MovieSceneScriptingFloatChannel}
          Scale.Y {type=MovieSceneScriptingFloatChannel}
          Shear.X {type=MovieSceneScriptingFloatChannel}
          Shear.Y {type=MovieSceneScriptingFloatChannel}
         ]
        }]
      }]
    }]}
 }
 TogglePresetsFullscreenMobile {
  binding_count=4
  length_seconds=0.3
  object=&self:TogglePresetsFullscreenMobile
  sequence {bindings [
    Categories {
     id=#10
     possessed_class=WBP_MPLocker_Categories_C
     tracks [{
       type=MovieSceneFloatTrack
       display_name=Render Opacity
       sections [{type=MovieSceneFloatSection, end_frame=3, start_frame=0, channels [None {type=MovieSceneScriptingFloatChannel, keys [[0.005,1],[3.0037,0]]}]}]
      }]
    }
    GridSlot_0 {
     id=#1
     parent_id=#11
     possessed_class=GridSlot
     tracks [{
       type=MovieSceneMarginTrack
       display_name=Padding
       sections [{
         type=MovieSceneMarginSection
         end_frame=6
         start_frame=0
         channels [
          Left {type=MovieSceneScriptingFloatChannel}
          Top {type=MovieSceneScriptingFloatChannel, keys [[0,0],[6,48]]}
          Right {type=MovieSceneScriptingFloatChannel}
          Bottom {type=MovieSceneScriptingFloatChannel, const [0,3]}
         ]
        }]
      }]
    }
    FortMPLocker_C_82 {
     id=#3
     possessed_class=FortMPLocker_C
     tracks [
      {
       type=MovieSceneDoubleTrack
       display_name=Presets Menu Size
       sections [{
         type=MovieSceneDoubleSection
         end_frame=6
         start_frame=0
         channels [None {type=MovieSceneScriptingDoubleChannel, thinned_from=3, keys [[0,1],[6.0003,0]]}]
        }]
      }
      {
       type=MovieSceneByteTrack
       display_name=Visibility
       sections [{type=MovieSceneByteSection, end_frame=6, start_frame=0, channels [None {type=MovieSceneScriptingByteChannel, keys [[0,4],[1,3],[6.0003,4]]}]}]
      }
     ]
    }
    PresetsWidget {
     id=#11
     possessed_class=WBP_MPLocker_Presets_C
     tracks [{
       type=MovieSceneFloatTrack
       display_name=Render Opacity
       sections [{type=MovieSceneFloatSection, end_frame=4, start_frame=0, channels [None {type=MovieSceneScriptingFloatChannel, keys [[0,0.3],[4.0043,1]]}]}]
      }]
    }
   ]}
 }
]
graphs [
 AdjustPresetsCamera {nodes [
   n1 = AdjustPresetsCamera!() -> (then, FloatProgress:double=~) {is=title K2Node_FunctionEntry}
   > SetSize!(
    execute
    self:Horizontal Box Slot Object Reference=SlotAsHorizontalBoxSlot!(Widget:Widget Object Reference=Get Camera_Spacer_Right() -> (Camera_Spacer_Right:Spacer Object Reference)) -> (ReturnValue:Horizontal Box Slot Object Reference)
    InSize_Value:float=Lerp(A=0.1, B=0, Alpha=@n1)
    InSize_SizeRule:ESlateSizeRule Enum=Fill
   ) -> (then)
   n2 = Lerp(A=725, B=1024, Alpha=@n1)
   n3 = Get SizeBox_PreviewWindow() -> (SizeBox_PreviewWindow:Size Box Object Reference)
   > SetWidthOverride(self=@n3, InWidthOverride=SelectFloat(A=Get LastWidthOverride() -> (LastWidthOverride:double), B=@n2, bPickA="float >= float"(A=@n1, B=1)))
   > SetPadding!(
    execute
    self:Horizontal Box Slot Object Reference=SlotAsHorizontalBoxSlot(Widget=@n3)
    InPadding_Left:float=0
    InPadding_Top:float=Lerp(A=64, B=0, Alpha=@n1)
    InPadding_Right:float=0
    InPadding_Bottom:float=0
   ) -> (then)
   > Set CameraWidth(CameraWidth:double=@n2) -> (Output_Get:double)
  ]}
 AnimStarted {nodes [AnimStarted!() -> (then, AnimLength:double=~) {is=title K2Node_FunctionEntry}]}
 BP_GetDesiredFocusTarget {nodes [
   Get Desired Focus Target()
   n1 = Get UIKit_SideNavigation() -> (UIKit_SideNavigation:WBP UIKit Side Navigation Object Reference)
   n2 = Get SelectedCategoryGroup!(self:Fort Item Selection VM Object Reference=Get FortItemSelectionVM() -> (FortItemSelectionVM:Fort Item Selection VM Object Reference)) -> (SelectedCategoryGroup:Fort Locker Category Group VM Object Reference)
   n3 = Get Categories() -> (Categories:WBP MPLocker Categories Object Reference)
   > Return Node!(
    execute
    ReturnValue:Widget Object Reference=Select(
     Option 0:Widget Object Reference=GetSideNavigationButtonBaseByViewModel!(self:WBP UIKit Side Navigation Base Object Reference=@n1, TabViewModel:MVVM Base Viewmodel Object Reference=@n2) -> (ButtonBase:Common Button Base Object Reference)
     Option 1:Widget Object Reference=GetFocusTarget!(self:WBP MPLocker Categories Object Reference=@n3) -> (ReturnValue:Widget Object Reference)
     Index=GetItemsAreEquippable!(self) -> (bItemsAreEquippable:bool)
    ) -> (ReturnValue:Widget Object Reference)
   )
  ]}
 BP_GetDesiredInputConfig {nodes [
   Get Desired Input Config() {is=title K2Node_FunctionEntry}
   > Return Node(ReturnValue:UIInput Config Structure=Make UIInput Config!(
     bIgnoreMoveInput:bool=False
     bIgnoreLookInput:bool=False
     InputMode:ECommonInputMode Enum=All
     MouseCaptureMode:EMouseCaptureMode Enum=CaptureDuringMouseDown
     MouseLockMode:EMouseLockMode Enum=DoNotLock
     bHideCursorDuringViewportCapture:bool=True
    ) -> (UIInputConfig:UIInput Config Structure) {is=title K2Node_MakeStruct})
  ]}
 Bind Preview Actions {nodes [
   Bind Preview Actions() {is=title K2Node_FunctionEntry}
   n1 = Get Hovered Category() -> (Hovered Category:WBP UIKit Item Card Category Object Reference)
   > Is Valid(InputObject=@n1)
   > Bind Preview Actions!(
    execute
    self:WBP UIKit Item Card Base Object Reference=@n1
    Preview Widget:Fort Item Preview Widget Object Reference=Get FortItemPreviewWidget() -> (FortItemPreviewWidget:WBP Item Preview Object Reference)
    Context Menu:WBP UIKit Context Menu Anchor Button Base Object Reference=Get ContextMenu!(self:WBP UIKit Item Card Category Object Reference=@n1) -> (ContextMenu:WBP UIKit Context Menu Anchor Button Object Reference)
    Item VM:Fort Item VM Object Reference=Get FortItemVM!(self:WBP UIKit Item Card Category Object Reference=@n1) -> (FortItemVM:Fort Item VM Object Reference)
    Include Special Actions:bool=true
    Include Toggle Actions:bool=true
   ) -> (then) {is=title K2Node_CallFunction}
  ]}
 EventGraph {nodes [
   n1 = "On Swipe Down (PresetsSwipePanel)"() {is=title K2Node_ComponentBoundEvent}
   n2 = Get Is Presets Active() -> (Is Presets Active:bool)
   n3 = Branch(execute=[@n1.then, @n73.then], Condition=@n2)
   n4 = Branch(execute=@n3.else, Condition=GetItemsAreEquippable())
   SwitchScreen_Presets!(
    execute=[@n4.then, @n6.then]
    self
    IsPresetsActive:bool=NOT Boolean(A=Get Is Presets Active())
    NoAnimations:bool=false
    IsSaveMode:bool=false
   ) -> (then)
   "On Character Saved (Categories)"() {is=title K2Node_ComponentBoundEvent}
   > SwitchScreen_Presets(IsPresetsActive=true, NoAnimations=false, IsSaveMode=true)
   n5 = "On Swipe Up (PresetsSwipePanel)"() {is=title K2Node_ComponentBoundEvent}
   n6 = Branch(execute=[@n5.then, @n75.then], Condition=@n2)
   n7 = "On Category Item Setup (Categories)"!() -> (OutputDelegate, then, CategoryItemEntryObject:User Widget Object Reference) {is=title K2Node_ComponentBoundEvent}
   > n8 = Cast To WBP_UIKit_ItemCard_Category!(execute, Object=@n7.CategoryItemEntryObject) -> (then, CastFailed, AsWBP UIKit Item Card Category:WBP UIKit Item Card Category Object Reference) {is=title K2Node_DynamicCast}
   n9 = OnCategoryCardHovered!() -> (OutputDelegate, then, Button:Common Button Base Object Reference) {is=title K2Node_CustomEvent}
   Bind Event to On Hovered(execute=@n8.then, self:Common Button Base Object Reference=@n8, Delegate=@n9) {is=title K2Node_AddDelegate}
   "On Preview Item Shown (FortItemPreviewWidget)"!() -> (OutputDelegate, then, ItemDefinition:Fort Account Item Definition Object Reference) {is=title K2Node_ComponentBoundEvent}
   > Bind Preview Actions!(execute, self:Self Object Reference) -> (then)
   n10 = "On Clicked (Button_BackToTop)"!() -> (OutputDelegate, then, Button:Common Button Base Object Reference) {is=title K2Node_ComponentBoundEvent}
   > ScrollListToTop(self:WBP MPLocker Presets Object Reference=Get PresetsWidget() -> (PresetsWidget:WBP MPLocker Presets Object Reference))
   > Fire Event Element Interacted(Element=@n10.Button)
   "On Category Edit Styles (Categories)"() {is=title K2Node_ComponentBoundEvent}
   n11 = Get Intro()
   > Branch(Condition=NOT Boolean(A=IsAnimationPlaying(InAnimation=@n11)))
   > n12 = "Play Animation with Finished event (legacy)"!(
    execute:Exec
    Widget:User Widget Object Reference=Self-Reference()
    InAnimation:Widget Animation Object Reference=Get Intro()
    StartAtTime:float=0
    NumLoopsToPlay:int=1
    PlayMode:EUMGSequencePlayMode Enum=Reverse
    PlaybackSpeed:float=1
   ) -> (then:Exec, Result:UMGSequence Player Object Reference, Finished:Exec) {is=title K2Node_PlayAnimation}
   > Set ActiveScreen(ActiveScreen:ELockerScreenList Enum=NewEnumerator4) -> (Output_Get:ELockerScreenList Enum=NewEnumerator0)
   > Handle3DCameraTransition()
   n13 = PushPrimaryContentClassAsync!(
    execute:Exec=@n12.Finished
    WidgetClass:Common Activatable Widget Class Reference=/Locker/MPL/EditStyles/WBP_MPL_VariantSelector.WBP_MPL_VariantSelector_C
   ) -> (then:Exec, OnPrimaryContentPushed:Exec, ActivatableWidget:Common Activatable Widget Object Reference) {is=title K2Node_AsyncAction}
   n14 = Cast To WBP_MPL_VariantSelector!(execute=@n13.OnPrimaryContentPushed, Object=@n13) -> (then, CastFailed, AsWBP MPL Variant Selector:WBP MPL Variant Selector Object Reference) {is=title K2Node_DynamicCast}
   n15 = Get FortVariantSelectorVM!(self:WBP MPL Variant Selector Object Reference=@n14) -> (FortVariantSelectorVM:Fort Variant Selector VM Object Reference)
   > DismissVariantChannels(self:Fort Variant Selector VM Object Reference=@n15)
   > UpdateVariantChannels(self:Fort Variant Selector VM Object Reference=@n15)
   "On Clicked (Button_RandomToggle)"!() -> (OutputDelegate, then, Button:Common Button Base Object Reference) {is=title K2Node_ComponentBoundEvent}
   > UpdateRandomButton()
   > RandomButtonClicked_Event(self:WBP MPLocker Presets Object Reference=Get PresetsWidget())
   n16 = "Is Modal Active (Categories)"!() -> (OutputDelegate, then, IsModalActive:bool) {is=title K2Node_ComponentBoundEvent}
   > Set Is Modal Active(Is Modal Active:bool=@n16.IsModalActive) -> (Output_Get:bool)
   "On Clicked (PresetsHitbox)"!() -> (OutputDelegate, then, Button:Common Button Base Object Reference) {is=title K2Node_ComponentBoundEvent}
   > Branch(Condition=GetItemsAreEquippable())
   > SwitchScreen_Presets(IsPresetsActive=true, NoAnimations=false, IsSaveMode=false)
   n17 = "On Clicked (Button_SeeLegoKits)"!() -> (OutputDelegate, then, Button:Common Button Base Object Reference) {is=title K2Node_ComponentBoundEvent}
   > n18 = Fire Event Element Interacted(Element=@n17.Button)
   n19 = GetItemSelectionVMConfig!() -> (VMConfig:Fort Locker Category Config Object Reference)
   n20 = Get CustomCategoryEditor!(self:Fort Locker Category Config Object Reference=@n19) -> (CustomCategoryEditor:Common Activatable Widget Soft Class Reference)
   n21 = Branch(
    execute=[@n18.then, @n30.then, @n39.then]
    Condition=AND Boolean(
     A=IsValidSoftClassReference!(SoftClassReference:"Object Soft Class Reference (by ref)"=@n20) -> (ReturnValue:bool)
     B=Get bDisplayAsModal!(self:Fort Locker Category Config Object Reference=@n19) -> (bDisplayAsModal:bool)
    )
   )
   > QueueModalWidgetSoftClassAsync!(execute:Exec, SoftWidgetClass:Common Activatable Widget Soft Class Reference=@n20) -> (then:Exec, OnModalWidgetQueued:Exec, ActivatableWidget:Common Activatable Widget Object Reference) {is=title K2Node_AsyncAction}
   n22 = IsAnimationPlaying(InAnimation=Get Intro())
   n23 = GetItemSelectionVMConfig()
   n24 = IsValidSoftClassReference(SoftClassReference=Get CustomCategoryEditor(self=@n23))
   n25 = Get bShowIfEmpty!(self:Fort Locker Category Config Object Reference=@n23) -> (bShowIfEmpty:bool)
   Branch(execute=@n21.else, Condition=AND Boolean(A=NOT Boolean(A=@n22), B=OR Boolean(A=@n25, B=@n24)))
   > n26 = "Play Animation with Finished event (legacy)"(Widget=Self-Reference(), InAnimation=Get Intro(), StartAtTime=0, NumLoopsToPlay=1, PlayMode=Reverse, PlaybackSpeed=1)
   > Set ActiveScreen(ActiveScreen=NewEnumerator3) -> (Output_Get=NewEnumerator0)
   > Handle3DCameraTransition()
   n27 = Get CustomCategoryEditor(self=GetItemSelectionVMConfig())
   n28 = Branch(execute=@n26.Finished, Condition=IsValidSoftClassReference(SoftClassReference=@n27))
   > PushPrimaryContentSoftClassAsync!(execute:Exec, SoftWidgetClass:Common Activatable Widget Soft Class Reference=@n27) -> (then:Exec, OnPrimaryContentPushed:Exec, ActivatableWidget:Common Activatable Widget Object Reference) {is=title K2Node_AsyncAction}
   PushPrimaryContentClassAsync(execute=@n28.else, WidgetClass=&/WBP_MPLocker_Items.WBP_MPLocker_Items_C)
   n29 = "Is Hovered Over Tile (Categories)"!() -> (OutputDelegate, then, IsTileHovered:bool) {is=title K2Node_ComponentBoundEvent}
   > UpdateOptionsButtonOverlayVisibility!(execute, self, ShowButtonContainer:bool=@n29.IsTileHovered) -> (then)
   "On Clear All Bangs (Categories)"() {is=title K2Node_ComponentBoundEvent}
   > ClearAllNewItemFlags(self:Global Fort Locker VM Object Reference=Get FortLockerVM() -> (FortLockerVM:Global Fort Locker VM Object Reference))
   n30 = "On Category Button Clicked (Categories)"() {is=title K2Node_ComponentBoundEvent}
   "On Hovered (PresetsHitbox)"!() -> (OutputDelegate, then, Button:Common Button Base Object Reference) {is=title K2Node_ComponentBoundEvent}
   n31 = Get PresetsPeek() -> (PresetsPeek:Widget Animation Object Reference)
   > PlayAnimationForward(InAnimation=@n31)
   "On Unhovered (PresetsHitbox)"!() -> (OutputDelegate, then, Button:Common Button Base Object Reference) {is=title K2Node_ComponentBoundEvent}
   > PlayAnimationReverse(InAnimation=@n31)
   "On Clicked (Button_Equip)"!() -> (OutputDelegate, then, Button:Common Button Base Object Reference) {is=title K2Node_ComponentBoundEvent}
   > HandleTouchLoadoutEquip(self:WBP MPLocker Presets Object Reference=Get PresetsWidget())
   "On Clicked (Button_Back)"!() -> (OutputDelegate, then, Button:Common Button Base Object Reference) {is=title K2Node_ComponentBoundEvent}
   > n32 = Branch(Condition=Get IsTabsExpanded() -> (IsTabsExpanded:bool))
   > OnCategoriesTabUnhover()
   SwitchScreen_Presets(execute=@n32.else, IsPresetsActive=false, NoAnimations=false, IsSaveMode=false)
   "On Clicked (Button_Options)"!() -> (OutputDelegate, then, Button:Common Button Base Object Reference) {is=title K2Node_ComponentBoundEvent}
   > n33 = Branch(Condition=Get Is Presets Active())
   > Toggle Selected Context!(
    execute
    self:WBP MPLocker Presets Object Reference=Get PresetsWidget()
    SelectedWidget:WBP UIKit Preset Card Object Reference=Get Hovered Preset Button() -> (Hovered Preset Button:WBP UIKit Preset Card Object Reference)
   ) -> (then)
   n34 = Get Categories()
   Branch(
    execute=@n33.else
    Condition=HasAnyUserFocusedDescendants!(self:Widget Object Reference=Get WBP_CategorySet!(self:WBP MPLocker Categories Object Reference=@n34) -> (WBP_CategorySet:WBP Category Set Object Reference)) -> (ReturnValue:bool)
   )
   > ToggleContextMenuIfSelected(self:WBP UIKit Item Card Category Object Reference=Get Hovered Category())
   "Preset Saved (PresetsWidget)"!() -> (OutputDelegate, then, EndInteraction:bool) {is=title K2Node_ComponentBoundEvent}
   > SwitchScreen_Presets(IsPresetsActive=false, NoAnimations=false, IsSaveMode=false)
   n35 = OnSideNavigationSelectionChanged!() -> (OutputDelegate, then, TabID:Name, InnerTabID:Name, TabViewModel:MVVM Base Viewmodel Object Reference) {is=title K2Node_CustomEvent}
   > UpdateSelectedCategoryViewModel!(execute, self, Tab VM:MVVM Base Viewmodel Object Reference=@n35.TabViewModel, TabNameID:Name=@n35.TabID) -> (then)
   n36 = OnSideNavigationTabClicked!() -> (OutputDelegate, then, TabID:Name, IsTabContainerExpanded:bool=~, TabViewModel:MVVM Base Viewmodel Object Reference) {is=title K2Node_CustomEvent}
   > n37 = Cast To FortLockerCategoryGroupVM!(execute, Object=@n36.TabViewModel) -> (then, CastFailed, AsFort Locker Category Group VM:Fort Locker Category Group VM Object Reference) {is=title K2Node_DynamicCast}
   > n38 = UpdateSelectedCategoryViewModel(Tab VM=@n37, TabNameID=@n36.TabID)
   PressedTab(execute=[@n38.then, @n53.then])
   n39 = OnCollectableTabsPressed() {is=title K2Node_CustomEvent}
   n40 = PresetHovered!() -> (OutputDelegate, then, Preset Button:WBP UIKit Preset Card Object Reference) {is=title K2Node_CustomEvent}
   > Set Hovered Preset Button(Hovered Preset Button:WBP UIKit Preset Card Object Reference=@n40.Preset Button) -> (Output_Get:WBP UIKit Preset Card Object Reference)
   n41 = Get FortTypedModularLoadoutVM!(self:WBP UIKit Preset Card Object Reference=@n40.Preset Button) -> (FortTypedModularLoadoutVM:Fort Typed Modular Loadout VM Object Reference)
   > Is Valid(InputObject=@n41)
   > SetCurrentPreviewLoadout!(
    execute
    self:Fort Item Preview Widget Object Reference=Get FortItemPreviewWidget()
    LoadoutVM:Fort Typed Modular Loadout VM Object Reference=@n41
   ) -> (then)
   > UpdateRandomButton()
   OnCategoryOptionContextMenuOpened() {is=title K2Node_CustomEvent}
   > Set IsCategoryContextMenuOpened(IsCategoryContextMenuOpened:bool=true) -> (Output_Get:bool)
   OnCategoryOptionContextMenuClosed() {is=title K2Node_CustomEvent}
   > Set IsCategoryContextMenuOpened(IsCategoryContextMenuOpened=false)
   n42 = OnChangeBackToTopButtonVisibility!() -> (OutputDelegate, then, New Visibility:ESlateVisibility Enum=Visible) {is=title K2Node_CustomEvent}
   > SetVisibility(self=Get Button_BackToTop() -> (Button_BackToTop:WBP UIKit Button Quiet Object Reference), InVisibility=@n42.New Visibility)
   n43 = OnWidgetActivatedEvent() {is=title K2Node_CustomEvent}
   SetFrontEndCamera!(
    execute=[@n71.then, @n43.then]
    self:Fort Front End Context Object Reference=Get FortFrontEndContext() -> (ReturnValue:Fort Front End Context Object Reference) {is=title K2Node_GetSubsystem}
    NewState:EFrontEndCamera Enum=CosmeticDisplay
    TransitionParams:View Target Transition Params Structure=(BlendTime=0,BlendFunction=VTBlend_Cubic,BlendExp=2,bLockOutgoing=False)
   ) -> (then)
   n44 = Create Event()
   n45 = Get Categories()
   > Bind Event to On Option Context Menu Opened(self:WBP MPLocker Categories Object Reference=@n45, Delegate=@n44) {is=title K2Node_AddDelegate}
   n46 = Create Event()
   > Bind Event to On Option Context Menu Closed(self:WBP MPLocker Categories Object Reference=@n45, Delegate=@n46) {is=title K2Node_AddDelegate}
   > ActivateWidget(self=Get FortItemPreviewWidget())
   > PlayAnimationForward(self=Self-Reference(), InAnimation=Get Intro())
   > Set ActiveScreen(ActiveScreen=NewEnumerator1) -> (Output_Get=NewEnumerator0)
   > SelectFirstSlot()
   > Handle3DCameraTransition()
   > SpawnSound2D!(
    execute
    Sound:Sound Base Object Reference=Get LockerCategoryWhooshSound() -> (LockerCategoryWhooshSound:Sound Base Object Reference)
    VolumeMultiplier:float=1
    PitchMultiplier:float=1
    StartTime:float=0
    ConcurrencySettings:Sound Concurrency Object Reference
    bPersistAcrossLevelTransition:bool=false
    bAutoDestroy:bool=true
   ) -> (then, ReturnValue:Audio Component Object Reference)
   > SetHasPrompts!(
    execute
    self:WBP Action Bar Container Object Reference=Get WBP_ActionBarContainer() -> (WBP_ActionBarContainer:WBP Action Bar Container Object Reference)
    HasPrompts:bool=true
   ) -> (then)
   > UpdateOptionsButtonOverlayVisibility(ShowButtonContainer=true)
   > Set Locker Item Widgets Visibility() {is=title K2Node_CallFunction}
   > SelectCorrespondentSideNavButton()
   n47 = Get FortGlobalUIContext!() -> (ReturnValue:Fort Global UIContext Object Reference) {is=title K2Node_GetBlueprintContextEx}
   > Branch(Condition=NOT Boolean(A=IsInZone!(self:Fort Global UIContext Object Reference=@n47) -> (ReturnValue:bool)))
   > n48 = StartListeningToEvent!(
    execute:Exec
    Target:Gameplay Event Router Component Object Reference=GetGlobalEventRouter()
    EventType:Script Struct Object Reference=/Script/FortniteGame.GameplayMessageDualBladeMenuVisible
    Context:Object Reference
    EventBubblingRule:EEventBubblingRule Enum
    None:%=Get GameplayMessageRouter() -> (ReturnValue:Gameplay Message Router Object Reference) {is=title K2Node_GetSubsystem}
   ) -> (then:Exec, AsyncTaskProxy:Async Action Start Listening To Event Object Reference, OnEventReceived:Exec, Payload_bBladeMenuVisible:bool=~)
   > Set DualBladeEventWatch(DualBladeEventWatch:Async Action Start Listening To Event Object Reference=@n48) -> (Output_Get:Async Action Start Listening To Event Object Reference)
   Set IsDualBladeOpen(execute=@n48.OnEventReceived, IsDualBladeOpen:bool=@n48.Payload_bBladeMenuVisible) -> (Output_Get:bool)
   n49 = AnimStartedInputRestorer!() -> (OutputDelegate, then, AnimLength:double=~) {is=title K2Node_CustomEvent}
   > Restore Focus!(In:Exec, DelayTime:double=@n49.AnimLength) {is=title K2Node_MacroInstance}
   n50 = InputChanged!() -> (OutputDelegate, then, bNewInputType:ECommonInputType Enum) {is=title K2Node_CustomEvent}
   n51 = Get Is Presets Active()
   > Branch(Condition=AND Boolean(A="Equal (Enum)"(A=@n50.bNewInputType, B=Touch), B=NOT Boolean(A=@n51)))
   > UpdateOptionsButtonOverlayVisibility(ShowButtonContainer=false)
   n52 = OnSideNavigationInnerTabClicked!() -> (OutputDelegate, then, TabID:Name, InnerTabID:Name, TabViewModel:MVVM Base Viewmodel Object Reference) {is=title K2Node_CustomEvent}
   > n53 = UpdateSelectedCategoryViewModel(Tab VM=@n52.TabViewModel, TabNameID=@n52.TabID)
   OnCategoriesTabUnhover() {is=title K2Node_CustomEvent}
   n54 = Get CommonInputSubsystem!(PlayerController:Player Controller Object Reference=GetOwningPlayer()) -> (ReturnValue:Common Input Object Reference) {is=title K2Node_GetSubsystemFromPC}
   > n55 = Branch(Condition="Equal (Enum)"(A=GetCurrentInputType(self=@n54), B=Gamepad))
   > PressedTab()
   Set IsTabsExpanded(execute=[@n61.then, @n55.else], IsTabsExpanded:bool=false) -> (Output_Get:bool)
   Unbinds() {is=title K2Node_CustomEvent}
   n56 = Get PresetsWidget()
   > Unbind all Events from Loadout Load!(execute:Exec, self:WBP MPLocker Presets Object Reference=@n56) -> (then:Exec) {is=title K2Node_ClearDelegate}
   > Unbind all Events from Preset Display Item!(execute:Exec, self:WBP MPLocker Presets Object Reference=@n56) -> (then:Exec) {is=title K2Node_ClearDelegate}
   > Unbind all Events from Change Back to Top Button Visibility!(execute:Exec, self:WBP MPLocker Presets Object Reference=@n56) -> (then:Exec) {is=title K2Node_ClearDelegate}
   > Unbind all Events from On Item Selection Changed!(
    execute:Exec
    self:List View Object Reference=Get TileView_Presets!(self:WBP MPLocker Presets Object Reference=@n56) -> (TileView_Presets:Fort Tile View Object Reference)
   ) -> (then:Exec) {is=title K2Node_ClearDelegate}
   > Unbind all Events from Anim Started!(execute:Exec, self:Fort MPLocker Object Reference) -> (then:Exec) {is=title K2Node_ClearDelegate}
   > SetSideNavigationBindingState!(execute, self, bShouldAddBindings:bool=false) -> (then)
   PressedTab() {is=title K2Node_CustomEvent}
   n57 = Get CommonInputSubsystem(PlayerController=GetOwningPlayer())
   > n58 = Branch(Condition="Equal (Enum)"(A=GetCurrentInputType(self=@n57), B=Gamepad))
   > n59 = Branch(Condition=Is Menu Focused!() -> (hasFocus:bool))
   > n60 = SetFocus(self=GetDesiredFocusTarget!(self:Common Activatable Widget Object Reference) -> (ReturnValue:Widget Object Reference))
   CloseTabs(execute=[@n60.then, @n59.else])
   Handle Collectable Tabs Clicked(execute=@n58.else)
   n61 = CloseTabs() {is=title K2Node_CustomEvent}
   HandleCategoryGroupChanged!() -> (OutputDelegate, then, Selected Category Group:Fort Locker Category Group VM Object Reference) {is=title K2Node_CustomEvent}
   > SelectFirstSlot()
   > Set Locker Item Widgets Visibility()
   > Set Collectable Preview Widget Position() {is=title K2Node_CallFunction}
   n62 = OnItemSelectionChanged!() -> (OutputDelegate, then, Item:Object Reference, bIsSelected:bool=~) {is=title K2Node_CustomEvent}
   > Branch(Condition=@n62.bIsSelected)
   > n63 = Cast To FortTypedModularLoadoutVM!(execute, Object=@n62.Item) -> (then, CastFailed, AsFort Typed Modular Loadout VM:Fort Typed Modular Loadout VM Object Reference) {is=title K2Node_DynamicCast}
   > Set LastSelectedLoadout(LastSelectedLoadout:Fort Typed Modular Loadout VM Object Reference=@n63) -> (Output_Get:Fort Typed Modular Loadout VM Object Reference)
   > Update Buttons Visibility!(execute, self, Typed Modular Loadout:Fort Typed Modular Loadout VM Object Reference=@n63) -> (then)
   Event Construct()
   > Unbinds() {is=title K2Node_CallFunction}
   > n64 = Sequence!(execute) -> (then_0, then_1, then_2:Exec)
   n65 = Get PresetsWidget()
   > Bind Event to Preset Display Item(self:WBP MPLocker Presets Object Reference=@n65, Delegate=Create Event()) {is=title K2Node_AddDelegate}
   > Bind Event to Loadout Load(self:WBP MPLocker Presets Object Reference=@n65, Delegate=Create Event()) {is=title K2Node_AddDelegate}
   > Bind Event to Change Back to Top Button Visibility(self:WBP MPLocker Presets Object Reference=@n65, Delegate=Create Event()) {is=title K2Node_AddDelegate}
   > SetSideNavigationBindingState(bShouldAddBindings=true)
   Bind Event to Anim Started(execute=@n64.then_1, self:Fort MPLocker Object Reference, Delegate=Create Event()) {is=title K2Node_AddDelegate}
   > Bind Event to On Item Selection Changed(self:List View Object Reference=Get TileView_Presets(self=Get PresetsWidget()), Delegate=Create Event()) {is=title K2Node_AddDelegate}
   n66 = Create Event()
   n67 = Get CommonInputSubsystem() {is=title K2Node_GetSubsystem}
   Bind Event to On Input Method Changed(execute=@n64.then_2, self=@n67, Delegate=@n66)
   Event PreConstruct()
   > n68 = Branch(Condition=IsMobileGame())
   n69 = Get TogglePresetsFullscreenMobile() -> (TogglePresetsFullscreenMobile:Widget Animation Object Reference)
   > PlayAnimation(InAnimation=@n69, StartAtTime=GetEndTime(self=@n69), PlayMode=Reverse)
   > SetMobileLayout()
   n70 = Get TogglePresetsFullscreen() -> (TogglePresetsFullscreen:Widget Animation Object Reference)
   PlayAnimation(execute=@n68.else, InAnimation=@n70, StartAtTime=GetEndTime(self=@n70), PlayMode=Reverse)
   > SetGridSlotPadding!(
    execute
    Widget:Widget Object Reference=Get VerticalBox_0!(self:WBP MPLocker Categories Object Reference=Get Categories()) -> (VerticalBox_0:Vertical Box Object Reference)
    Padding_Left:float=0
    Padding_Top:float=24
    Padding_Right:float=0
    Padding_Bottom:float=0
   ) -> (then)
   Event On Activated()
   > n71 = Update Flow Rider Sub View()
   Event On Deactivated()
   > Unbind Event from On Option Context Menu Opened(self:WBP MPLocker Categories Object Reference=@n45, Delegate=@n44) {is=title K2Node_RemoveDelegate}
   > Unbind Event from On Option Context Menu Closed(self:WBP MPLocker Categories Object Reference=@n45, Delegate=@n46) {is=title K2Node_RemoveDelegate}
   > SetCurrentPreviewItem!(execute, self:Fort Item Preview Widget Object Reference=Get FortItemPreviewWidget(), ItemVM:Fort Item VM Object Reference) -> (then)
   > DeactivateWidget(self=Get FortItemPreviewWidget())
   > SwitchScreen_Presets(IsPresetsActive=false, NoAnimations=true, IsSaveMode=false)
   > StopListeningToEvent(self:Async Action Start Listening To Event Object Reference=Get DualBladeEventWatch() -> (DualBladeEventWatch:Async Action Start Listening To Event Object Reference))
   > Set IsDualBladeOpen(IsDualBladeOpen=false)
   Event Destruct()
   > Unbind Event from On Input Method Changed(self=@n67, Delegate=@n66)
   > SetSideNavigationBindingState(bShouldAddBindings=false)
   Mouse Wheel Down!() -> (Pressed:Exec, Released:Exec, Key:Key Structure) {is=title K2Node_InputKey, bConsumeInput=false, bOverrideParentBinding=false, inputKey=MouseScrollDown}
   n72 = Is Scrolling Available!() -> (CanScroll:bool=~) {is=title K2Node_MacroInstance}
   > n73 = Branch(Condition=@n72)
   Mouse Wheel Up!() -> (Pressed:Exec, Released:Exec, Key:Key Structure) {is=title K2Node_InputKey, bConsumeInput=false, bOverrideParentBinding=false, inputKey=MouseScrollUp}
   n74 = GetScrollOffset!(self:List View Base Object Reference=Get TileView_Presets(self=Get PresetsWidget())) -> (ReturnValue:float)
   > n75 = Branch(Condition=AND Boolean(A=@n72, B="Equal (Float)"(A=@n74)))
   Bind Preview Actions(execute=@n77.then)
   SetIsEnabled(self=Get Button_Options() -> (Button_Options:WBP UIKit Button Quiet Object Reference), bInIsEnabled=false)
   SetIsEnabled(self=Get Button_Options(), bInIsEnabled=true)
   "Disable Options Hint (PresetsWidget)"() {is=title K2Node_ComponentBoundEvent}
   "Enable Options Hint (PresetsWidget)"() {is=title K2Node_ComponentBoundEvent}
   Reroute Node(InputPin:Wildcard) -> (OutputPin:Wildcard)
   n76 = Cast To WBP_UIKit_ItemCard_Category(execute=@n9.then, Object=@n9.Button)
   > n77 = Set Hovered Category(Hovered Category:WBP UIKit Item Card Category Object Reference=@n76) -> (Output_Get:WBP UIKit Item Card Category Object Reference)
  ]}
 GetItemSelectionVMConfig {nodes [
   GetItemSelectionVMConfig() {is=title K2Node_FunctionEntry}
   n1 = Get SelectedCategory!(self:Fort Item Selection VM Object Reference=Get FortItemSelectionVM()) -> (SelectedCategory:Fort Locker Category Item VM Object Reference)
   > Return Node!(
    execute
    VMConfig:Fort Locker Category Config Object Reference=GetConfig!(self:Fort Locker Category VM Object Reference=Get Category!(self:Fort Locker Category Item VM Object Reference=@n1) -> (Category:Fort Locker Category VM Object Reference)) -> (ReturnValue:Fort Locker Category Config Object Reference)
   )
  ]}
 GetItemsAreEquippable {nodes [
   GetItemsAreEquippable() {is=title K2Node_FunctionEntry}
   n1 = Get FortItemSelectionVM()
   > n2 = Is Valid(InputObject=@n1)
   Return Node!(execute=[@n4.Is Not Valid, @n2.Is Not Valid], bItemsAreEquippable:bool=false)
   n3 = Get SelectedCategoryGroup(self=@n1)
   > n4 = Is Valid(InputObject=@n3)
   > Return Node(bItemsAreEquippable=GetIsEquippable!(self:Fort Locker Category Item VM Object Reference=GetFirstCategoryItem!(self:Fort Locker Category Group VM Object Reference=@n3) -> (ReturnValue:Fort Locker Category Item VM Object Reference)) -> (ReturnValue:bool))
  ]}
 Handle Category Selected {nodes [
   n1 = Handle Category Selected!() -> (then, InputPin:Name) {is=title K2Node_FunctionEntry}
   n2 = Get FortLockerVM()
   > Is Valid(InputObject=@n2)
   > Set SelectedCategoryGroup!(
    execute
    SelectedCategoryGroup:Fort Locker Category Group VM Object Reference=FindCategoryGroupByName!(self:Global Fort Locker VM Object Reference=@n2, Name:Name=@n1) -> (ReturnValue:Fort Locker Category Group VM Object Reference)
    self:Fort Item Selection VM Object Reference=Get FortItemSelectionVM()
   ) -> (then, Output_Get:Fort Locker Category Group VM Object Reference)
   > SelectFirstSlot()
   > Set Locker Item Widgets Visibility()
   > Set Collectable Preview Widget Position()
  ]}
 Handle Collectable Tabs Clicked {nodes [
   Handle Collectable Tabs Clicked() {is=title K2Node_FunctionEntry}
   > n1 = Branch(Condition=GetItemsAreEquippable() {is=title K2Node_CallFunction})
   > Return Node!(execute)
   OnCollectableTabsPressed(execute=@n1.else) {is=title K2Node_CallFunction}
  ]}
 Handle3DCameraTransition {nodes [
   n1 = "Not Equal (Enum)"!(A:ELockerScreenList Enum=Get ActiveScreen() -> (ActiveScreen:ELockerScreenList Enum=NewEnumerator0), B:ELockerScreenList Enum=NewEnumerator1) -> (ReturnValue:bool)
   n2 = Branch(Condition="Equal (Boolean)"(A:bool=Get CameraIsLeft() -> (CameraIsLeft:bool), B:bool=@n1) -> (ReturnValue:bool=~) {is=title K2Node_PromotableOperator})
   PlayAnimationForward(
    execute=@n2.else
    InAnimation=Select(
     Option 0:Widget Animation Object Reference=Get CameraSlideIn() -> (CameraSlideIn:Widget Animation Object Reference)
     Option 1:Widget Animation Object Reference=Get CameraSlideOut() -> (CameraSlideOut:Widget Animation Object Reference)
     Index=@n1
    ) -> (ReturnValue:Widget Animation Object Reference)
   )
   > Set CameraIsLeft(CameraIsLeft:bool=@n1) -> (Output_Get:bool)
   Handle3DCameraTransition() {is=title K2Node_FunctionEntry}
  ]}
 Is Menu Focused {nodes [
   Is Menu Focused() {is=title K2Node_FunctionEntry}
   n1 = Get IsDualBladeOpen() -> (IsDualBladeOpen:bool)
   n2 = Get Is Modal Active() -> (Is Modal Active:bool)
   > Return Node!(execute, hasFocus:bool=AND Boolean(A=NOT Boolean(A=@n1), B=NOT Boolean(A=@n2)))
  ]}
 IsScrollingAvailable {nodes [
   Inputs!()
   n1 = Get IsCategoryContextMenuOpened() -> (IsCategoryContextMenuOpened:bool)
   n2 = IsAnimationPlaying(InAnimation=Get Intro())
   n3 = IsHovered(self=Get FortItemPreviewWidget())
   Outputs!(CanScroll:bool=AND Boolean!(
     A=Is Menu Focused() {is=title K2Node_CallFunction}
     B=Get bIsActive() -> (bIsActive:bool)
     C:bool=NOT Boolean(A=@n3)
     D:bool=NOT Boolean(A=@n1)
     E:bool=NOT Boolean(A=@n2)
    ) -> (ReturnValue))
  ]}
 Loadout Loaded {nodes [
   Loadout Loaded() {is=title K2Node_FunctionEntry}
   > Navigation_UpFromPresets!(execute, self, Navigation:EUINavigation Enum=Up) -> (then, ReturnValue:Widget Object Reference)
  ]}
 Navigation_DownToPresets {nodes [
   Navigation_DownToPresets!() -> (then, Navigation:EUINavigation Enum) {is=title K2Node_FunctionEntry}
   > SwitchScreen_Presets(IsPresetsActive=true, NoAnimations=false, IsSaveMode=false)
   > Return Node!(execute, ReturnValue:Widget Object Reference)
  ]}
 Navigation_GoToTabs {nodes [
   Navigation_GoToTabs!() -> (then, Navigation:EUINavigation Enum) {is=title K2Node_FunctionEntry}
   > UpdateOptionsButtonOverlayVisibility(ShowButtonContainer=false)
   n1 = Get FortItemSelectionVM()
   > n2 = Cast To WBP_UIKit_ButtonGeneric_Base!(execute, Object=GetSideNavigationButtonBaseByViewModel(self=Get UIKit_SideNavigation(), TabViewModel=Get SelectedCategoryGroup(self=@n1))) -> (then, CastFailed, AsWBP UIKit Button Generic Base:WBP UIKit Button Generic Base Object Reference) {is=title K2Node_DynamicCast}
   > Return Node(ReturnValue=@n2)
  ]}
 Navigation_GoToTiles {nodes [
   Navigation_GoToTiles!() -> (then, Navigation:EUINavigation Enum) {is=title K2Node_FunctionEntry}
   > UpdateOptionsButtonOverlayVisibility(ShowButtonContainer=true)
   n1 = Get PresetsWidget()
   > Return Node(ReturnValue=Select(
     Option 0:Widget Object Reference=GetDesiredFocusTarget()
     Option 1:Widget Object Reference=Get TileView_Presets(self=@n1)
     Index=Get Is Presets Active()
    ) -> (ReturnValue:Widget Object Reference))
  ]}
 Navigation_UpFromPresets {nodes [
   Navigation_UpFromPresets!() -> (then, Navigation:EUINavigation Enum) {is=title K2Node_FunctionEntry}
   > SwitchScreen_Presets(IsPresetsActive=false, NoAnimations=false, IsSaveMode=false)
   > Return Node()
  ]}
 PopulateCategories {nodes [
   n1 = PopulateCategories!() -> (then, CategoryGroupCollections:Array of Fort Locker Category Group Collection VM Object References) {is=title K2Node_FunctionEntry}
   > ClearSideNavigationTabs(self:WBP UIKit Side Navigation Base Object Reference=Get UIKit_SideNavigation())
   > Set LastArrayIndex(LastArrayIndex:int=int - int(A=Length(TargetArray:Array of Fort Locker Category Group Collection VM Object References=@n1))) -> (Output_Get:int=0)
   > n2 = For Each Loop(Array:Array of Fort Locker Category Group Collection VM Object References=@n1) -> (Array Element:Fort Locker Category Group Collection VM Object Reference)
   > Set CurrentArrayIndex(CurrentArrayIndex:int=@n2.Array Index) -> (Output_Get:int=0)
   > Set CurrentCategoryViewModel(CurrentCategoryViewModel:Fort Locker Category Group Collection VM Object Reference=@n2) -> (Output_Get:Fort Locker Category Group Collection VM Object Reference)
   > n3 = Branch(Condition=TextIsEmpty(InText=GetDisplayName!(self:Fort Locker Category Group Collection VM Object Reference=@n2) -> (ReturnValue:"Text (by ref)")))
   > n4 = For Each Loop(Array:Array of Fort Locker Category Group VM Object References=Get CategoryGroups!(self:Fort Locker Category Group Collection VM Object Reference=Get CurrentCategoryViewModel() -> (CurrentCategoryViewModel:Fort Locker Category Group Collection VM Object Reference)) -> (CategoryGroups:Array of Fort Locker Category Group VM Object References)) -> (Array Element:Fort Locker Category Group VM Object Reference)
   n5 = GetUniqueName!(self:Fort Locker Category Group VM Object Reference=@n4) -> (ReturnValue:Name=None)
   > AddSideNavigationTab!(
    execute
    self:WBP UIKit Side Navigation Base Object Reference=Get UIKit_SideNavigation()
    TabViewModel:MVVM Base Viewmodel Object Reference=@n4
    TabID:Name=@n5
    Index:int=-1
   ) -> (then)
   > Set CurrentSubGroupID(CurrentSubGroupID:Name=@n5) -> (Output_Get:Name=None)
   n6 = Get CurrentArrayIndex() -> (CurrentArrayIndex:int=0)
   n7 = Get LastArrayIndex() -> (LastArrayIndex:int=0)
   Branch(execute=[@n16.Completed, @n4.Completed], Condition=NOT Boolean(A="integer >= integer"(A=@n6, B=@n7)))
   n8 = Get CurrentSubGroupID() -> (CurrentSubGroupID:Name=None)
   n9 = "Not Equal (Name)"(A=@n8, B=LoadoutArchetype_Vehicles) -> (ReturnValue=~)
   n10 = "Not Equal (Name)"(A=@n8, B=LoadoutSchema_Platform) -> (ReturnValue=~)
   n11 = Get LastSubGroupID() -> (LastSubGroupID:Name=None)
   > Branch(Condition=AND Boolean!(A="Not Equal (Name)"(A=@n11, B=@n8) -> (ReturnValue=~), B=AND Boolean(A=@n9, B=@n10)) -> (ReturnValue))
   > Set LastSubGroupID(LastSubGroupID:Name=@n8) -> (Output_Get:Name=None)
   n12 = Get UIKit_SideNavigation()
   > AddDividerByIndex!(
    execute
    self:WBP UIKit Tab Group Base Object Reference=@n12
    TabIndex:int=int - int(A=GetTabCount!(self:Common Tab List Widget Base Object Reference=@n12) -> (ReturnValue:int=0))
    AddAfter:bool=true
   ) -> (then)
   n13 = Get CurrentCategoryViewModel()
   n14 = GetUniqueName(self:Fort Locker Category Group Collection VM Object Reference=@n13) -> (ReturnValue=None)
   n15 = Get UIKit_SideNavigation()
   AddSideNavigationTab(execute=@n3.else, self=@n15, TabViewModel=@n13, TabID=@n14, Index=-1)
   > n16 = For Each Loop(Array=Get CategoryGroups(self=@n13))
   > AddSideNavigationInnerTab!(
    execute
    self:WBP UIKit Side Navigation Base Object Reference=@n15
    TabViewModel:MVVM Base Viewmodel Object Reference=@n16
    TabID:Name=@n14
    InnerTabID:Name=GetUniqueName(self:Fort Locker Category Group VM Object Reference=@n16) -> (ReturnValue=None)
    Index:int=-1
   ) -> (then)
   > Set CurrentSubGroupID(CurrentSubGroupID=@n14) -> (Output_Get=None)
  ]}
 PushModal {nodes [
   n1 = PushModal!() -> (then, ActivatableWidget:Common Activatable Widget Object Reference) {is=title K2Node_FunctionEntry}
   > TryPushWidgetAsModalPanel!(execute, PushWidget:User Widget Object Reference=@n1) -> (then)
  ]}
 RestoreFocus {nodes [
   n1 = Inputs!() -> (In:Exec, DelayTime:double=~)
   > Delay(Duration=@n1)
   > Set bIsEnabled(bIsEnabled:bool=true) -> (Output_Get:bool)
   Outputs!()
  ]}
 SelectCorrespondentSideNavButton {nodes [
   SelectCorrespondentSideNavButton() {is=title K2Node_FunctionEntry}
   n1 = Get FortItemSelectionVM()
   > Is Valid(InputObject=@n1)
   n2 = Get SelectedCategoryGroup(self=@n1)
   > Is Valid(InputObject=@n2)
   > SelectSideNavigationViewModel!(
    execute
    self:WBP UIKit Side Navigation Base Object Reference=Get UIKit_SideNavigation()
    TabViewModel:MVVM Base Viewmodel Object Reference=@n2
    bSuppressClickFeedback:bool=true
   ) -> (then, Success:bool)
  ]}
 SelectFirstSlot {nodes [
   SelectFirstSlot() {is=title K2Node_FunctionEntry}
   n1 = Get FortItemSelectionVM()
   > Is Valid(InputObject=@n1)
   n2 = Get SelectedCategoryGroup(self=@n1)
   > Is Valid(InputObject=@n2)
   n3 = Get FortItemSelectionVM()
   n4 = Get CategoryItems!(self:Fort Locker Category Group VM Object Reference=@n2) -> (CategoryItems:Array of Fort Locker Category Item VM Object References)
   > n5 = Branch(Condition=Contains Item(
     TargetArray:Array of Fort Locker Category Item VM Object References=@n4
     ItemToFind:"Fort Locker Category Item VM Object Reference (by ref)"=GetPreviewCategory!(self:Fort Item Selection VM Object Reference=@n3) -> (ReturnValue:Fort Locker Category Item VM Object Reference)
    ))
   n6 = Get FortItemPreviewWidget()
   n7 = GetPreviewCategory(self=Get FortItemSelectionVM())
   > SetCurrentPreviewItem(self=@n6, ItemVM=Get EquippedItem!(self:Fort Locker Category Item VM Object Reference=@n7) -> (EquippedItem:Fort Item VM Object Reference))
   > SetSlotIndexOverride!(
    execute
    self:Fort Item Preview Widget Object Reference=@n6
    SlotIndex:int=Get SlotIndex!(self:Fort Locker Category Item VM Object Reference=@n7) -> (SlotIndex:int=0)
   ) -> (then)
   n8 = "Get (a copy)"(Array:Array of Fort Locker Category Item VM Object References=@n4) -> (Output:Fort Locker Category Item VM Object Reference)
   n9 = Get FortItemSelectionVM()
   Set SelectedCategory!(execute=@n5.else, SelectedCategory:Fort Locker Category Item VM Object Reference=@n8, self:Fort Item Selection VM Object Reference=@n9) -> (then, Output_Get:Fort Locker Category Item VM Object Reference)
   > Set HoveredCategory!(execute, HoveredCategory:Fort Locker Category Item VM Object Reference=@n8, self:Fort Item Selection VM Object Reference=@n9) -> (then, Output_Get:Fort Locker Category Item VM Object Reference)
  ]}
 Set Category Item Widgets Visibility {nodes [
   n1 = Set Category Item Widgets Visibility!() -> (then, Visibility:ESlateVisibility Enum=Visible) {is=title K2Node_FunctionEntry}
   n2 = Get Categories()
   > n3 = Set Visibility(
    Visibility=@n1
    self=Get Text_SubHeader!(self:WBP MPLocker Heading Object Reference=Get WBP_MPLocker_Heading!(self:WBP MPLocker Categories Object Reference=@n2) -> (WBP_MPLocker_Heading:WBP MPLocker Heading Object Reference)) -> (Text_SubHeader:Common Text Object Reference)
   )
   > n4 = Set Visibility(Visibility=@n3, self=Get WBP_CategorySet(self=@n2))
   > n5 = Set Visibility(
    Visibility=@n4
    self=Get Button_SavePresets!(self:WBP MPLocker Categories Object Reference=@n2) -> (Button_SavePresets:WBP Locker Button Object Reference)
   )
   > n6 = Set Visibility(
    Visibility=@n5
    self=Get Button_Options!(self:WBP MPLocker Categories Object Reference=@n2) -> (Button_Options:WBP UIKit Context Menu Anchor Button Object Reference)
   )
   > Set Visibility(
    Visibility=@n6
    self=Get WPB_ItemDescription!(self:WBP MPLocker Categories Object Reference=@n2) -> (WPB_ItemDescription:WBP UIKit Item Description MPL Object Reference)
   )
  ]}
 Set Collectable Preview Widget Position {nodes [
   Set Collectable Preview Widget Position() {is=title K2Node_FunctionEntry}
   n1 = GetItemsAreEquippable()
   > SetGridSlotPadding!(
    execute
    Widget=Get Widget_CameraFrame() -> (Widget_CameraFrame:Size Box Object Reference)
    Padding:Margin Structure=Select(
     Option 0:Margin Structure=Make Margin(Left=0, Top=250, Right=500, Bottom=0)
     Option 1:Margin Structure=Make Margin(Left=0, Top=0, Right=0, Bottom=0)
     Index=@n1
    ) -> (ReturnValue:Margin Structure)
   ) -> (then)
   > n2 = Branch(Condition=@n1)
   n3 = Get SelectedCategory(self=Get FortItemSelectionVM())
   > Is Valid(InputObject=@n3)
   > UpdateCategoryPreviewScale!(execute, self, CategoryToPreview:Fort Locker Category Item VM Object Reference=@n3) -> (then)
   SetWidthOverride(execute=@n2.else, self=Get SizeBox_PreviewWindow(), InWidthOverride=Make Literal Float(Value=1300))
  ]}
 Set Locker Item Widgets Visibility {nodes [
   Set Locker Item Widgets Visibility() {is=title K2Node_FunctionEntry}
   > n1 = Sequence()
   n2 = GetItemsAreEquippable()
   > Set Category Item Widgets Visibility!(
    execute
    self
    Visibility:ESlateVisibility Enum=Select(Option 0:ESlateVisibility Enum=Hidden, Option 1:ESlateVisibility Enum=Visible, Index=@n2) -> (ReturnValue:ESlateVisibility Enum)
   ) -> (then) {is=title K2Node_CallFunction}
   Set Visibility(
    execute=@n1.then_1
    Visibility=Select(Option 0=Visible, Option 1=Collapsed, Index=@n2)
    self=Get Button_SeeLegoKits() -> (Button_SeeLegoKits:WBP UIKit Button Quiet Object Reference)
   )
   > SetHasPrompts(self=Get WBP_ActionBarContainer(), HasPrompts=NOT Boolean(A=@n2))
   > SetVisibility(self=Get Button_Options!() -> (Button_Options:WBP UIKit Button Quiet Object Reference), InVisibility=Collapsed)
   n3 = Set Visibility(execute=@n1.then_2, Visibility=Select(Option 0=Collapsed, Option 1=HitTestInvisible, Index=@n2), self=Get PresetsWidget())
   > Set Visibility(Visibility=@n3, self=Get Camera_Spacer_Left() -> (Camera_Spacer_Left:Spacer Object Reference))
   > Return Node!(execute)
  ]}
 SetCameraPosition {nodes [
   n1 = SetCameraPosition!() -> (then, CameraPosition:double=~) {is=title K2Node_FunctionEntry}
   > SetSize(self=SlotAsHorizontalBoxSlot(Widget=Get Camera_Spacer_Left()), InSize_Value=Lerp(A=1, B=0, Alpha=@n1), InSize_SizeRule=Fill)
   > SetSize(self=SlotAsHorizontalBoxSlot(Widget=Get Camera_Spacer_Right()), InSize_Value=Lerp(A=0, B=1, Alpha=@n1), InSize_SizeRule=Fill)
  ]}
 SetMobileLayout {nodes [
   SetMobileLayout() {is=title K2Node_FunctionEntry}
   > Branch(Condition=IsMobileGame())
   n1 = Get SizeBox_0() -> (SizeBox_0:Size Box Object Reference)
   > ClearWidthOverride(self:Size Box Object Reference=@n1)
   > Set HAlign!(
    execute
    HAlign:EHorizontalAlignment Enum=HAlign_Fill
    self:Safe Zone Slot Object Reference=SlotAsSafeBoxSlot!(Widget:Widget Object Reference=@n1) -> (ReturnValue:Safe Zone Slot Object Reference)
   ) -> (then, Output_Get:EHorizontalAlignment Enum=HAlign_Fill)
   n2 = SlotAsGridSlot(Widget=Get CategoryTabs_NavWrapper() -> (CategoryTabs_NavWrapper:Grid Panel Object Reference))
   > SetPadding!(execute, self:Grid Slot Object Reference=@n2, InPadding:Margin Structure=Make Margin(Left=48, Top=176, Right=48, Bottom=0)) -> (then)
   > SetVerticalAlignment(self:Grid Slot Object Reference=@n2, InVerticalAlignment=VAlign_Top)
   > SetPadding(self=SlotAsGridSlot(Widget=Get WBP_ActionBarContainer()), InPadding=Make Margin(Left=0, Top=0, Right=58, Bottom=24))
   n3 = Get UIKit_SideNavigation()
   > Set IconSize!(execute, IconSize:Vector 2D Structure=(X=48,Y=48), self:WBP UIKit Tab Group Base Object Reference=@n3) -> (then, Output_Get:Vector 2D Structure)
   > Set BackplateCornerRadius!(execute, BackplateCornerRadius:E_UI_BackplateCornerRadius Enum=NewEnumerator0, self:WBP UIKit Tab Group Base Object Reference=@n3) -> (then, Output_Get:E_UI_BackplateCornerRadius Enum=NewEnumerator0)
   n4 = Get ScrollBox_Tabs!(self:WBP UIKit Tab Group Base Object Reference=@n3) -> (ScrollBox_Tabs:Scroll Box Object Reference)
   > SetScrollBarVisibility!(execute, self:Scroll Box Object Reference=@n4, NewScrollBarVisibility:ESlateVisibility Enum=Visible) -> (then)
   > Set!(
    execute
    ScrollbarPadding_Left:float=8
    ScrollbarPadding_Top:float=16
    ScrollbarPadding_Right:float=16
    ScrollbarPadding_Bottom:float=16
    self:Scroll Box Object Reference=@n4
   ) -> (then, Output_Get:Margin Structure)
   > SetSidesToPad!(
    execute
    self:Safe Zone Object Reference=Get SafeZone() -> (SafeZone:Safe Zone Object Reference)
    InPadLeft:bool=true
    InPadRight:bool=true
    InPadTop:bool=true
    InPadBottom:bool=false
   ) -> (then)
   n5 = GetViewportSize!() -> (ReturnValue_X:double, ReturnValue_Y:double)
   > Branch(Condition="float >= float"(A=float / float(A=@n5, B=@n5.ReturnValue_Y), B=1.34))
   > SetHeightOverride(self=Get SizeBox_Mobile() -> (SizeBox_Mobile:Size Box Object Reference), InHeightOverride=736)
  ]}
 SetPresetsMenuSize {nodes [
   n1 = SetPresetsMenuSize!() -> (then, IsFullscreen:double=~) {is=title K2Node_FunctionEntry}
   n2 = Get PresetsWidget()
   > SetRowFill!(
    execute
    self:Grid Panel Object Reference=Get ContentMover!(self:WBP MPLocker Presets Object Reference=@n2) -> (ContentMover:Grid Panel Object Reference)
    RowIndex:int=0
    Coefficient:float=@n1
   ) -> (then)
   > AdjustPresetsCamera!(execute, self, FloatProgress:double=@n1) -> (then) {is=title K2Node_CallFunction}
   > SetSize!(
    execute
    self:Spacer Object Reference=Get TopPadding!(self:WBP MPLocker Presets Object Reference=@n2) -> (TopPadding:Spacer Object Reference)
    InSize_X:double=0
    InSize_Y:double=Lerp(A=128, B=0, Alpha=@n1)
   ) -> (then)
  ]}
 SetSideNavAllEnabled {nodes [
   n1 = SetSideNavAllEnabled!() -> (then, Enabled:bool=~) {is=title K2Node_FunctionEntry}
   n2 = Get UIKit_SideNavigation()
   > SetVisibility(self=@n2, InVisibility=Select(Option 0=HitTestInvisible, Option 1=SelfHitTestInvisible, Index=@n1))
   > n3 = For Each Loop(Array:Array of Widget Object References=GetAllChildren(self=Get TabsContainer!(self:WBP UIKit Tab Group Base Object Reference=@n2) -> (TabsContainer:Grid Panel Object Reference))) -> (Array Element:Widget Object Reference)
   > Set RenderOpacity!(
    execute
    RenderOpacity:float=Select(Option 0:float=0.4, Option 1:float=1, Index=@n1) -> (ReturnValue:float)
    self:Widget Object Reference=@n3
   ) -> (then, Output_Get:float)
   ChangeEnabledStyle!(execute, self:WBP Categories Button Nested Category Object Reference, Enabled:bool=false) -> (then)
  ]}
 SetSideNavigationBindingState {nodes [
   n1 = SetSideNavigationBindingState!() -> (then, bShouldAddBindings:bool=~) {is=title K2Node_FunctionEntry}
   > n2 = Get!(execute:Exec) -> (then:Exec, else:Exec, UIKit_SideNavigation:WBP UIKit Side Navigation Object Reference)
   > n3 = Branch(Condition=@n1)
   n4 = Create Event()
   > Bind Event to On Side Navigation Selection Changed(self:WBP UIKit Side Navigation Base Object Reference=@n2, Delegate=@n4) {is=title K2Node_AddDelegate}
   n5 = Create Event()
   > Bind Event to On Side Navigation Tab Clicked(self:WBP UIKit Side Navigation Base Object Reference=@n2, Delegate=@n5) {is=title K2Node_AddDelegate}
   n6 = Create Event()
   > Bind Event to On Side Navigation Inner Tab Clicked(self:WBP UIKit Side Navigation Base Object Reference=@n2, Delegate=@n6) {is=title K2Node_AddDelegate}
   Unbind Event from On Side Navigation Selection Changed(execute=@n3.else, self:WBP UIKit Side Navigation Base Object Reference=@n2, Delegate=@n4) {is=title K2Node_RemoveDelegate}
   > Unbind Event from On Side Navigation Tab Clicked(self:WBP UIKit Side Navigation Base Object Reference=@n2, Delegate=@n5) {is=title K2Node_RemoveDelegate}
   > Unbind Event from On Side Navigation Inner Tab Clicked(self:WBP UIKit Side Navigation Base Object Reference=@n2, Delegate=@n6) {is=title K2Node_RemoveDelegate}
  ]}
 ShufflePresetsActive {nodes [
   n1 = ShufflePresetsActive!() -> (then, ShuffleEnabled:bool=~) {is=title K2Node_FunctionEntry}
   n2 = Get Is Presets Active()
   > Branch(Condition=@n2)
   > UpdateRandomButton()
   > Branch(Condition=AND Boolean(A=@n1, B=@n2))
   > SwitchScreen_Presets(IsPresetsActive=false, NoAnimations=false, IsSaveMode=false)
  ]}
 SwitchScreen_Presets {nodes [
   n1 = SwitchScreen_Presets!() -> (then, IsPresetsActive:bool=~, NoAnimations:bool=~, IsSaveMode:bool=~) {is=title K2Node_FunctionEntry}
   > Branch(Condition=Get bIsEnabled() -> (bIsEnabled:bool))
   > Set Is Save Mode(Is Save Mode:bool=@n1.IsSaveMode) -> (Output_Get:bool)
   > Set bIsEnabled(bIsEnabled=false)
   > n2 = Set Is Presets Active(Is Presets Active:bool=@n1) -> (Output_Get:bool)
   > n3 = Branch(Condition=@n2)
   > SetVisibility(self=Get Categories(), InVisibility=HitTestInvisible)
   > ActivateWidget(self=Get PresetsWidget())
   > SetSaveMode!(execute, self:WBP MPLocker Presets Object Reference=Get PresetsWidget(), SaveMode:bool=Get Is Save Mode() -> (Is Save Mode:bool)) -> (then)
   > SetVisibility(self=Get PresetsWidget(), InVisibility=SelfHitTestInvisible)
   > SetVisibility(self=Get PresetsHitbox() -> (PresetsHitbox:WBP UIKit Button Regular Object Reference), InVisibility=HitTestInvisible)
   > SetSideNavAllEnabled!(execute, self, Enabled:bool=false) -> (then) {is=title K2Node_CallFunction}
   > K2Node_Composite_0 = Mobile Forward Preset Anim!(execute) -> (if not mobile:Exec, then:Exec) {is=title K2Node_Composite}
   SetVisibility(
    execute=[@n6.then, @K2Node_Composite_0.then]
    self=[
     Get Overlay_EquipVisibility() -> (Overlay_EquipVisibility:Overlay Object Reference)
     Get Overlay_RandomToggle() -> (Overlay_RandomToggle:Overlay Object Reference)
     Get Overlay_BackToTopVisibility() -> (Overlay_BackToTopVisibility:Overlay Object Reference)
     Get Button_Back() -> (Button_Back:WBP UIKit Button Quiet Object Reference)
    ]
    InVisibility=Visible
   )
   > UpdateRandomButton()
   > UpdateOptionsButtonOverlayVisibility(ShowButtonContainer=true)
   n4 = Get TogglePresetsFullscreen()
   n5 = GetEndTime(self=@n4)
   PlayAnimation(
    execute=@K2Node_Composite_0.if not mobile
    InAnimation=@n4
    StartAtTime=SelectFloat(A=@n5, B=0, bPickA=Get NoAnimations() -> (NoAnimations:bool))
   )
   > PlayAnimationForward(InAnimation=Get PresetsPeek())
   > n6 = Call Anim Started!(execute:Exec, self:Fort MPLocker Object Reference, AnimLength:double=@n5) -> (then:Exec) {is=title K2Node_CallDelegate}
   SetVisibility(execute=@n3.else, self=Get Categories(), InVisibility=SelfHitTestInvisible)
   > DeactivateWidget(self=Get PresetsWidget())
   > SetSaveMode(self=Get PresetsWidget(), SaveMode=false)
   > SetVisibility(self=Get PresetsWidget(), InVisibility=HitTestInvisible)
   > SetVisibility(self=Get PresetsHitbox(), InVisibility=SelfHitTestInvisible)
   > SetSideNavAllEnabled(Enabled=true)
   > K2Node_Composite_1 = Mobile Reverse Preset Anim!(execute) -> (if not mobile:Exec, then:Exec) {is=title K2Node_Composite}
   SetVisibility(
    execute=[@n11.then, @K2Node_Composite_1.then]
    self=[Get Button_Back(), Get Overlay_RandomToggle(), Get Overlay_EquipVisibility(), Get Overlay_BackToTopVisibility()]
    InVisibility=Collapsed
   )
   n7 = Get FortItemPreviewWidget()
   n8 = GetPreviewCategory(self=Get FortItemSelectionVM())
   > SetCurrentPreviewItem(self=@n7, ItemVM=Get EquippedItem(self=@n8))
   > SetSlotIndexOverride(self=@n7, SlotIndex=Get SlotIndex(self=@n8) -> (SlotIndex=0))
   > SetVariantOverrides!(
    execute
    self:Fort Item Preview Widget Object Reference=@n7
    InVariantOverrides:Array of Mcp Variant Channel Info Structures=Make Array!() -> (Array:Array of Mcp Variant Channel Info Structures)
   ) -> (then)
   > Update Buttons Visibility(Typed Modular Loadout=Get LastSelectedLoadout() -> (LastSelectedLoadout:Fort Typed Modular Loadout VM Object Reference))
   > UpdateOptionsButtonOverlayVisibility(ShowButtonContainer=true)
   n9 = Get TogglePresetsFullscreen()
   n10 = GetEndTime(self=@n9)
   PlayAnimation(
    execute=@K2Node_Composite_1.if not mobile
    InAnimation=@n9
    StartAtTime=SelectFloat(A=@n10, B=0, bPickA=Get NoAnimations())
    PlayMode=Reverse
   )
   > PlayAnimationReverse(InAnimation=Get PresetsPeek())
   > n11 = Call Anim Started(AnimLength=@n10)
  ]}
 Mobile Forward Preset Anim {
  path=SwitchScreen_Presets.K2Node_Composite_0.Mobile Forward Preset Anim
  nodes [
   Inputs!() -> (execute:Exec)
   > n1 = Branch(Condition=IsMobileGame())
   n2 = Get TogglePresetsFullscreenMobile()
   n3 = GetEndTime(self=@n2)
   > PlayAnimation(InAnimation=@n2, StartAtTime=SelectFloat(A=@n3, B=0, bPickA=Get NoAnimations()))
   > PlayAnimationForward(InAnimation=Get PresetsPeek())
   > n4 = Call Anim Started(AnimLength=@n3)
   Outputs!(if not mobile:Exec=@n1.else, then:Exec=@n4.then)
  ]
 }
 Mobile Reverse Preset Anim {
  path=SwitchScreen_Presets.K2Node_Composite_1.Mobile Reverse Preset Anim
  nodes [
   Inputs()
   > n1 = Branch(Condition=IsMobileGame())
   n2 = Get TogglePresetsFullscreenMobile()
   n3 = GetEndTime(self=@n2)
   > PlayAnimation(InAnimation=@n2, StartAtTime=SelectFloat(A=@n3, B=0, bPickA=Get NoAnimations()), PlayMode=Reverse)
   > PlayAnimationReverse()
   > n4 = Call Anim Started(AnimLength=@n3)
   Outputs(if not mobile=@n1.else, then=@n4.then)
   Get PresetsPeek()
  ]
 }
 Update Buttons Visibility {nodes [
   n1 = Update Buttons Visibility!() -> (then, Typed Modular Loadout:Fort Typed Modular Loadout VM Object Reference) {is=title K2Node_FunctionEntry}
   > n2 = Is Valid(InputObject=@n1)
   n3 = Select(
    Option 0:ESlateVisibility Enum=Collapsed
    Option 1:ESlateVisibility Enum=SelfHitTestInvisible
    Index=Get bIsValid!(self:Fort Typed Modular Loadout VM Object Reference=@n1) -> (bIsValid:bool)
   ) -> (ReturnValue:ESlateVisibility Enum)
   > SetVisibility(self=Get Button_Equip() -> (Button_Equip:WBP UIKit Button Quiet Object Reference), InVisibility=@n3)
   n4 = NOT Boolean(A=Get Is Save Mode())
   n5 = Get bIsShuffle!(self:Fort Typed Modular Loadout VM Object Reference=@n1) -> (bIsShuffle:bool)
   > SetVisibility(
    self=Get Button_RandomToggle() -> (Button_RandomToggle:WBP UIKit Button Quiet Object Reference)
    InVisibility=Select(Option 0=Collapsed, Option 1=SelfHitTestInvisible, Index=AND Boolean(A=@n5, B=@n4))
   )
   > n6 = Branch(Condition=Get Is Presets Active())
   > SetVisibility(self=Get Button_Options(), InVisibility=@n3)
   SetVisibility(
    execute=[@n6.else, @n2.Is Not Valid]
    self=Get Button_Options()
    InVisibility=Select(Option 0=Collapsed, Option 1=Visible, Index=GetItemsAreEquippable())
   )
  ]}
 Update Flow Rider Sub View {nodes [
   Update Flow Rider Sub View() {is=title K2Node_FunctionEntry}
   n1 = Get SelectedCategoryGroup(self=Get FortItemSelectionVM())
   > "Fire Event Sub View Opened (Non-Localized)"!(
    execute
    FortPC:Fort Player Controller Object Reference=GetFortPlayerControllerFromActor!(Actor:Actor Object Reference=GetOwningPlayer()) -> (ReturnValue:Fort Player Controller Object Reference)
    ViewName:String="To String (Name)"(InName=GetUniqueName(self=@n1) -> (ReturnValue=None))
    AdditionalInfo:Flow Analytics Additional Info Structure
    bIsOverlay:bool=false
   ) -> (then)
  ]}
 UpdateCategoryPreviewScale {nodes [
   n1 = UpdateCategoryPreviewScale!() -> (then, CategoryToPreview:Fort Locker Category Item VM Object Reference) {is=title K2Node_FunctionEntry}
   > Is Valid(InputObject=@n1)
   n2 = GetSlotInfo!(self:Fort Locker Category Item VM Object Reference=@n1) -> (
    ReturnValue_CategoryName:Text
    ReturnValue_CategoryNamePlural:Text
    ReturnValue_CategoryNameWithPluralForms:Text
    ReturnValue_CategoryDescription:Text
    ReturnValue_DefaultEmptyTexture:Soft Object Reference
    ReturnValue_OverrideShuffleTexture:Texture 2D Soft Object Reference
    ReturnValue_CategoryIcon:Material Instance Object Reference
    ReturnValue_CategorySoundOverrides:Map of Gameplay Tag Structures to Sound Base Soft Object References
    ReturnValue_bShowInCategoryTab:bool=~
    ReturnValue_PreviewWindowScale:float=~
    ReturnValue_PreviewWindowOffset_X:double
    ReturnValue_PreviewWindowOffset_Y:double
    ReturnValue_PreviewWindowOffset_Z:double
    ReturnValue_bUsePluralCategoryInItemList:bool=~
   )
   n3 = float * float(A=@n2.ReturnValue_PreviewWindowScale, B=Get CameraWidth() -> (CameraWidth:double)) {is=title K2Node_PromotableOperator}
   > Set WidthOverride!(execute, WidthOverride:float=@n3, self:Size Box Object Reference=Get SizeBox_PreviewWindow()) -> (then, Output_Get:float)
   > Set LastWidthOverride(LastWidthOverride:double=@n3) -> (Output_Get:double)
   > SetRenderTranslation!(
    execute
    self=Get Widget_CameraFrame()
    Translation_X:double=@n2.ReturnValue_PreviewWindowOffset_X
    Translation_Y:double=@n2.ReturnValue_PreviewWindowOffset_Y
   ) -> (then)
  ]}
 UpdateOptionsButtonOverlayVisibility {nodes [
   n1 = UpdateOptionsButtonOverlayVisibility!() -> (then, ShowButtonContainer:bool=~) {is=title K2Node_FunctionEntry}
   n2 = Get Is Presets Active()
   n3 = GetOwningPlayer()
   > n4 = Branch(Condition=AND Boolean(A=IsUsingTouch(PlayerController=@n3), B=NOT Boolean(A=@n2)))
   > SetVisibility(self=Get Overlay_OptionVisibility() -> (Overlay_OptionVisibility:Overlay Object Reference), InVisibility=Collapsed)
   > SetHasPrompts(self=Get WBP_ActionBarContainer(), HasPrompts=false)
   SetHasPrompts(execute=@n4.else, self=Get WBP_ActionBarContainer(), HasPrompts=@n1)
   > SetVisibility(self=Get Overlay_OptionVisibility(), InVisibility=Select(Option 0=Collapsed, Option 1=Visible, Index=@n1))
   > Update Buttons Visibility(Typed Modular Loadout=Get LastSelectedLoadout()) {is=title K2Node_CallFunction}
  ]}
 UpdateRandomButton {nodes [
   UpdateRandomButton() {is=title K2Node_FunctionEntry}
   n1 = Get FortTypedModularLoadoutsVM() -> (FortTypedModularLoadoutsVM:Fort Typed Modular Loadouts VM Object Reference)
   > Is Valid(InputObject=@n1)
   n2 = Get bShuffleEnabled!(self:Fort Typed Modular Loadouts VM Object Reference=@n1) -> (bShuffleEnabled:bool)
   n3 = Get Button_RandomToggle()
   > SetText!(
    execute
    self:Fort CTAButton Object Reference=@n3
    ButtonText:"Text (by ref)"=SelectText(A="LOCTEXT(\"Random On\")", B="LOCTEXT(\"Random Off\")", bPickA=NOT Boolean(A=@n2))
   ) -> (then)
   n4 = Get ActiveLoadoutsCount!(self:Fort Typed Modular Loadouts VM Object Reference=@n1) -> (ActiveLoadoutsCount:int=0)
   > SetIsEnabled(self=@n3, bInIsEnabled=Select(Option 0:bool=true, Option 1:bool=false, Index="integer < integer"(A=@n4, B=2)) -> (ReturnValue:bool))
  ]}
 UpdateSelectedCategoryViewModel {nodes [
   n1 = UpdateSelectedCategoryViewModel!() -> (then, Tab VM:MVVM Base Viewmodel Object Reference, TabNameID:Name) {is=title K2Node_FunctionEntry}
   > n2 = Cast To FortLockerCategoryGroupVM(Object=@n1)
   n3 = Get FortItemSelectionVM()
   > n4 = Branch(Condition="Equal (Object)"(A=@n2, B=Get SelectedCategoryGroup(self=@n3)))
   n5 = Set SelectedCategoryGroup(execute=@n4.else, SelectedCategoryGroup=@n2, self=@n3)
   Update Flow Rider Sub View(execute=[@n19.then, @n5.then]) {is=title K2Node_CallFunction}
   n6 = Cast To FortLockerCategoryGroupCollectionVM!(execute=@n2.CastFailed, Object=@n1) -> (then, CastFailed, AsFort Locker Category Group Collection VM:Fort Locker Category Group Collection VM Object Reference) {is=title K2Node_DynamicCast}
   n7 = Get CategoryGroups(self=@n6)
   > n8 = Set CategoryGroupsLength(CategoryGroupsLength:int=Length(TargetArray:Array of Fort Locker Category Group VM Object References=@n7)) -> (Output_Get:int=0)
   > Branch(Condition="integer > integer"(A=@n8))
   n9 = Get SelectedCategoryGroup(self=@n3)
   > Is Valid(InputObject=@n9)
   n10 = "Get (a copy)"(Array:Array of Fort Locker Category Group VM Object References=@n7) -> (Output:Fort Locker Category Group VM Object Reference)
   > n11 = Branch(Condition="Equal (Object)"(A=@n9, B=@n10))
   n12 = Get UIKit_SideNavigation()
   n13 = Get TabNameID() -> (TabNameID:Name=None)
   n14 = GetUniqueName(self=@n9) -> (ReturnValue=None)
   n15 = Get UIKit_SideNavigation()
   n16 = Branch(
    execute=@n11.else
    Condition="integer < integer"(
     A=GetTabIndexByID!(self:UIKit Tab Group Object Reference=@n12, TabNameID:Name=@n13) -> (ReturnValue:int=0)
     B=GetTabIndexByID(self=@n15, TabNameID=@n14) -> (ReturnValue=0)
    )
   )
   n17 = "Get (a copy)"(Array=@n7, Dimension 1=int - int(A=Get CategoryGroupsLength() -> (CategoryGroupsLength:int=0)))
   > SelectSideNavigationViewModel(self=Get UIKit_SideNavigation(), TabViewModel=@n17, bSuppressClickFeedback=true)
   > n18 = Set SelectedCategoryGroup(SelectedCategoryGroup=@n17, self=@n3)
   Branch(execute=[@n18.then, @n20.then], Condition=Is Menu Focused())
   > n19 = SetFocus(self=GetSideNavigationSelectedButtonBase!(self:WBP UIKit Side Navigation Base Object Reference=Get UIKit_SideNavigation()) -> (Button:Common Button Base Object Reference))
   SelectSideNavigationViewModel(execute=@n16.else, self=Get UIKit_SideNavigation(), TabViewModel=@n10, bSuppressClickFeedback=true)
   > n20 = Set SelectedCategoryGroup(SelectedCategoryGroup=@n10, self=@n3)
  ]}
]
variables [
 ActiveScreen {type=ELockerScreenList, default=Overview}
 AnimStarted {}
 CameraIsLeft {type=boolean}
 CameraPosition {type=number}
 CameraWidth {type=number, default=1024}
 DualBladeEventWatch {type=AsyncAction_StartListeningToEvent, default=null}
 Hovered Category {type=WBP_UIKit_ItemCard_Category_C, default=null}
 Hovered Preset Button {type=WBP_UIKit_PresetCard_C, default=null}
 Is Modal Active {type=boolean}
 Is Presets Active {type=boolean}
 Is Save Mode {type=boolean}
 IsCategoryContextMenuOpened {type=boolean}
 IsDualBladeOpen {type=boolean}
 IsTabsExpanded {type=boolean}
 LastSelectedLoadout {type=FortTypedModularLoadoutVM, default=null}
 LastWidthOverride {type=number, default=1024}
 LockerCategoryWhooshSound {type=SoundBase, default=null}
 PresetsMenuSize {type=number, default=1}
 SoftActivation {type=boolean}
]
widgets {
 info {
  inheritedWidgetCount=1
  namedSlotCount=1
  parentClass=/Script/FortniteUI.AthenaCustomizationScreenBase
  rootWidgetClass=&SafeZone
  widgetCount=36
 }
 widgets [
  SafeZone {widgetClassPath=&SafeZone, bIsVariable=true, widget=SafeZone, properties {padBottom=false, padTop=false}}
  SizeBox_0 {
   widgetClassPath=&SizeBox
   bIsVariable=true
   parent=SafeZone
   widget=SizeBox_0
   properties {bOverride_WidthOverride=true, widthOverride=1920}
   slot {class=SafeZoneSlot, properties {hAlign=HAlign_Center}}
  }
  PresetsSwipePanel {
   widgetClassPath=/Script/FortniteUI.FortSwipePanel
   bIsVariable=true
   parent=SizeBox_0
   widget=PresetsSwipePanel
   properties {bConsumePointerInput=false, swipeThreshold=<0 175>}
   slot {class=SizeBoxSlot}
  }
  Grid_Main {
   widgetClassPath=&GridPanel
   bIsVariable=true
   parent=PresetsSwipePanel
   widget=Grid_Main
   properties {columnFill [0,1,1], rowFill [1,0.2]}
   slot {class=FortSwipePanelSlot}
  }
  PreviewInputBlocker {
   widgetClassPath=&Image
   parent=Grid_Main
   widget=PreviewInputBlocker
   properties {colorAndOpacity=<rgba 1 1 1 0>}
   slot {class=GridSlot, properties {column=1, layer=1, rowSpan=2}}
  }
  Widget_CameraFrame {
   widgetClassPath=&SizeBox
   bIsVariable=true
   parent=Grid_Main
   widget=Widget_CameraFrame
   slot {class=GridSlot, properties {column=1, columnSpan=2, rowSpan=2}}
  }
  HorizontalBox_0 {widgetClassPath=&HorizontalBox, parent=Widget_CameraFrame, widget=HorizontalBox_0, slot {class=SizeBoxSlot}}
  Camera_Spacer_Left {
   widgetClassPath=&Spacer
   bIsVariable=true
   parent=HorizontalBox_0
   widget=Camera_Spacer_Left
   slot {class=HorizontalBoxSlot, properties {size {sizeRule=Fill, value=1}}}
  }
  SizeBox_PreviewWindow {
   widgetClassPath=&SizeBox
   bIsVariable=true
   parent=HorizontalBox_0
   widget=SizeBox_PreviewWindow
   properties {bOverride_HeightOverride=true, bOverride_WidthOverride=true, heightOverride=1024, widthOverride=1024}
   slot {class=HorizontalBoxSlot, properties {horizontalAlignment=HAlign_Center, verticalAlignment=VAlign_Center}}
  }
  FortItemPreviewWidget {
   widgetClassPath=/FortniteCoreUI/Frontend/ItemPreview/WBP_ItemPreview
   bIsVariable=true
   parent=SizeBox_PreviewWindow
   widget=FortItemPreviewWidget
   properties {
    bRestrictZoomToWidget=true
    bSupportsActivationFocus=false
    itemViewObject=&self:WidgetTree.FortItemPreviewWidget.ItemViewObject
    visibility=Visible
    defaultPreviewContextTags {gameplayTags [{tagName=UI.ItemPreview.PreviewType.OnPawn}]}
   }
   slot {class=SizeBoxSlot, properties {padding=<ltrb 0 0 0 96>}}
  }
  Camera_Spacer_Right {
   widgetClassPath=&Spacer
   bIsVariable=true
   parent=HorizontalBox_0
   widget=Camera_Spacer_Right
   slot {class=HorizontalBoxSlot, properties {size {sizeRule=Fill, value=0}}}
  }
  WBP_ActionBarContainer {
   widgetClassPath=/Game/UI/Frontend/BottomBar/WBP_ActionBarContainer
   bIsVariable=true
   parent=Grid_Main
   widget=WBP_ActionBarContainer
   slot {
    class=GridSlot
    properties {columnSpan=3, horizontalAlignment=HAlign_Right, layer=5, row=1, verticalAlignment=VAlign_Bottom, padding=<ltrb 0 0 80 64>}
   }
  }
  Button_Bar {widgetClassPath=&HorizontalBox, namedSlotHost=WBP_ActionBarContainer, widget=Button_Bar}
  Overlay_BackToTopVisibility {
   widgetClassPath=&Overlay
   bIsVariable=true
   parent=Button_Bar
   widget=Overlay_BackToTopVisibility
   properties {visibility=Collapsed}
   slot {class=HorizontalBoxSlot}
  }
  VisibilityBorder_Touch_BackToTop {
   widgetClassPath=&CommonHardwareVisibilityBorder
   parent=Overlay_BackToTopVisibility
   widget=VisibilityBorder_Touch_BackToTop
   properties {
    padding=<ltrb 0 0 32 0>
    visibilityQuery {
     autoDescription=" ANY( Input.Touch )"
     tokenStreamVersion=0
     userDescription=""
     queryTokenStream [0,1,1,1,0]
     tagDictionary [{tagName=Input.Touch}]
    }
   }
   slot {class=OverlaySlot}
  }
  Button_BackToTop {
   widgetClassPath=/FortUILibrary/UIKit/Buttons/WBP_UIKit_Button_Quiet
   bIsVariable=true
   parent=VisibilityBorder_Touch_BackToTop
   widget=Button_BackToTop
   properties {
    bHideInputActionWithKeyboard=true
    text=Back To Top
    visibility=Hidden
    triggeringInputAction {dataTable=&InputActionDataTable, rowName=ContextPopupMenu}
   }
   slot {class=BorderSlot, properties {padding=<ltrb 0 0 32 0>}}
  }
  Overlay_EquipVisibility {
   widgetClassPath=&Overlay
   bIsVariable=true
   parent=Button_Bar
   widget=Overlay_EquipVisibility
   properties {visibility=Collapsed}
   slot {class=HorizontalBoxSlot}
  }
  VisibilityBorder_Touch {
   widgetClassPath=&CommonHardwareVisibilityBorder
   parent=Overlay_EquipVisibility
   widget=VisibilityBorder_Touch
   properties {
    padding=<ltrb 0 0 32 0>
    visibilityQuery {
     autoDescription=" ANY( Input.Touch )"
     tokenStreamVersion=0
     userDescription=""
     queryTokenStream [0,1,1,1,0]
     tagDictionary [{tagName=Input.Touch}]
    }
   }
   slot {class=OverlaySlot}
  }
  Button_Equip {
   widgetClassPath=&WBP_UIKit_Button_Quiet
   bIsVariable=true
   parent=VisibilityBorder_Touch
   widget=Button_Equip
   properties {
    bHideInputActionWithKeyboard=true
    text=Equip
    triggeringInputAction {dataTable=&InputActionDataTable, rowName=ContextPopupMenu}
   }
   slot {class=BorderSlot, properties {padding=<ltrb 0 0 32 0>}}
  }
  Overlay_RandomToggle {
   widgetClassPath=&Overlay
   bIsVariable=true
   parent=Button_Bar
   widget=Overlay_RandomToggle
   properties {visibility=Collapsed}
   slot {class=HorizontalBoxSlot}
  }
  VisibilityBorder_Touch_Random {
   widgetClassPath=&CommonHardwareVisibilityBorder
   parent=Overlay_RandomToggle
   widget=VisibilityBorder_Touch_Random
   properties {
    padding=<ltrb 0 0 32 0>
    visibilityQuery {
     autoDescription=" ANY( Input.Touch )"
     tokenStreamVersion=0
     userDescription=""
     queryTokenStream [0,1,1,1,0]
     tagDictionary [{tagName=Input.Touch}]
    }
   }
   slot {class=OverlaySlot}
  }
  Button_RandomToggle {
   widgetClassPath=&WBP_UIKit_Button_Quiet
   bIsVariable=true
   parent=VisibilityBorder_Touch_Random
   widget=Button_RandomToggle
   properties {
    bHideInputActionWithKeyboard=true
    text=Toggle Random
    triggeringInputAction {dataTable=&InputActionDataTable, rowName=ContextPopupMenu}
   }
   slot {class=BorderSlot, properties {padding=<ltrb 0 0 32 0>}}
  }
  Overlay_OptionVisibility {
   widgetClassPath=&Overlay
   bIsVariable=true
   parent=Button_Bar
   widget=Overlay_OptionVisibility
   properties {visibility=Collapsed}
   slot {class=HorizontalBoxSlot}
  }
  VisibilityBorder_TouchGamepad {
   widgetClassPath=&CommonHardwareVisibilityBorder
   parent=Overlay_OptionVisibility
   widget=VisibilityBorder_TouchGamepad
   properties {
    padding=<ltrb 0 0 32 0>
    visibilityQuery {
     autoDescription=" ANY( Input.Gamepad, Input.Touch )"
     tokenStreamVersion=0
     userDescription=""
     queryTokenStream [0,1,1,2,0,1]
     tagDictionary [{tagName=Input.Gamepad}, {tagName=Input.Touch}]
    }
   }
   slot {class=OverlaySlot}
  }
  Button_Options {
   widgetClassPath=&WBP_UIKit_Button_Quiet
   bIsVariable=true
   parent=VisibilityBorder_TouchGamepad
   widget=Button_Options
   properties {
    bHideInputActionWithKeyboard=true
    text=Options
    triggeringInputAction {dataTable=&InputActionDataTable, rowName=ContextPopupMenu}
   }
   slot {class=BorderSlot, properties {padding=<ltrb 0 0 32 0>}}
  }
  Button_SeeLegoKits {
   widgetClassPath=&WBP_UIKit_Button_Quiet
   bIsVariable=true
   parent=Button_Bar
   widget=Button_SeeLegoKits
   properties {text="See LEGO® Kits", triggeringInputAction {dataTable=&InputActionDataTable, rowName=GenericFaceButtonBottom}}
   slot {class=HorizontalBoxSlot}
  }
  Button_Back {
   widgetClassPath=&WBP_UIKit_Button_Quiet
   bIsVariable=true
   parent=Button_Bar
   widget=Button_Back
   properties {text=Back, visibility=Collapsed, triggeringInputAction {dataTable=&InputActionDataTable, rowName=GenericBack}}
   slot {class=HorizontalBoxSlot, properties {verticalAlignment=VAlign_Center}}
  }
  Button_BackToGame {
   widgetClassPath=&WBP_UIKit_Button_Quiet
   bInherited=true
   bIsVariable=true
   parent=Button_Bar
   widget=Button_BackToGame
   properties {
    text=Back To Game
    hoveredSlateSoundOverride {resourceObject=/Game/Sounds/Fort_UI_Sounds/FrontEnd/Locker/Locker_Button_Hover_Default_Cue}
    lockedHoveredSlateSoundOverride {resourceObject=&Locker_Button_Hover_Default_Cue}
    lockedPressedSlateSoundOverride {resourceObject=/Game/Sounds/Fort_UI_Sounds/FrontEnd/Locker/Locker_Button_Close_Pressed_Cue}
    pressedSlateSoundOverride {resourceObject=&Locker_Button_Close_Pressed_Cue}
    selectedHoveredSlateSoundOverride {resourceObject=&Locker_Button_Hover_Default_Cue}
    selectedPressedSlateSoundOverride {resourceObject=&Locker_Button_Close_Pressed_Cue}
    triggeringInputAction {dataTable=&InputActionDataTable, rowName=BackToGame}
   }
   slot {class=HorizontalBoxSlot, properties {padding=<ltrb 32 0 0 0>}}
  }
  CategoriesNavWrapper {
   widgetClassPath=&GridPanel
   parent=Grid_Main
   widget=CategoriesNavWrapper
   properties {navigation=&self:WidgetTree.CategoriesNavWrapper.WidgetNavigation_0, columnFill [1], rowFill [1]}
   slot {class=GridSlot, properties {column=1, columnSpan=2, layer=3}}
  }
  Categories {
   widgetClassPath=&/WBP_MPLocker_Categories
   bIsVariable=true
   parent=CategoriesNavWrapper
   widget=Categories
   properties {navigation=&self:WidgetTree.Categories.WidgetNavigation_0}
   slot {class=GridSlot}
  }
  Presets_NavWrapper {
   widgetClassPath=&GridPanel
   parent=Grid_Main
   widget=Presets_NavWrapper
   properties {
    navigation=&self:WidgetTree.Presets_NavWrapper.WidgetNavigation_0
    columnFill [1]
    renderTransform {angle=0, scale=<1 1>, shear=<0 0>, translation=<0 4>}
    rowFill [1]
   }
   slot {class=GridSlot, properties {column=1, columnSpan=2, layer=3, rowSpan=2}}
  }
  PresetsWidget {
   widgetClassPath=&/WBP_MPLocker_Presets
   bIsVariable=true
   parent=Presets_NavWrapper
   widget=PresetsWidget
   properties {renderOpacity=0.3, visibility=HitTestInvisible}
   slot {class=GridSlot}
  }
  CategoryTabs_NavWrapper {
   widgetClassPath=&GridPanel
   bIsVariable=true
   parent=Grid_Main
   widget=CategoryTabs_NavWrapper
   properties {navigation=&self:WidgetTree.CategoryTabs_NavWrapper.WidgetNavigation_0}
   slot {class=GridSlot, properties {layer=3, padding=<ltrb 80 180 56 0>}}
  }
  SizeBox_Mobile {
   widgetClassPath=&SizeBox
   bIsVariable=true
   parent=CategoryTabs_NavWrapper
   widget=SizeBox_Mobile
   slot {class=GridSlot, properties {column=1, verticalAlignment=VAlign_Center}}
  }
  UIKit_SideNavigation {
   widgetClassPath=/FortUILibrary/UIKit/Tabs/WBP_UIKit_SideNavigation
   bIsVariable=true
   parent=SizeBox_Mobile
   widget=UIKit_SideNavigation
   properties {
    buttonClass=&/WBP_CategoriesButton_NestedCategory.WBP_CategoriesButton_NestedCategory_C
    dividerClass=&/WBP_MPLocker_Sidebar_Divider.WBP_MPLocker_Sidebar_Divider_C
    dividerSpacing=10
    hidePrompts=true
    innerTabButtonClassOverride=&/WBP_CategoriesButton_NestedCategory_Hack.WBP_CategoriesButton_NestedCategory_Hack_C
    inputActions=Custom
    nextPromptHorizontalAlignment=HAlign_Right
    prevPromptHorizontalAlignment=HAlign_Left
    promptBackplateBrightness=Dark
    selectInnerTabOnContainerExpand=true
    tabSpacing=0
    usePromptBackplate=false
    buttonPadding=<ltrb 10 4 10 4>
    dividerPadding=<ltrb 17 0 17 0>
    innerTabGroupPadding=<ltrb 4 4 4 4>
    promptBackplateMinSize=<62 48>
    tabGroupPadding=<ltrb 0 0 0 0>
   }
   slot {class=SizeBoxSlot}
  }
  PresetsHitbox {
   widgetClassPath=&WBP_UIKit_Button_Regular
   bIsVariable=true
   parent=Grid_Main
   widget=PresetsHitbox
   properties {minHeight=56, navigation=&self:WidgetTree.PresetsHitbox.WidgetNavigation_0, renderOpacity=0, text=" "}
   slot {class=GridSlot, properties {column=1, layer=4, row=1}}
  }
 ]
}
