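# Standalone Item Shop screen widget, extending FortStandaloneItemShop, with a tab bar, content switcher, player info and touch close button.
# InitializeTabs registers only the Item Shop tab; Locker and V-Bucks registrations are unwired, and tab selection sets the switcher index.
# Camera-level load sets the Home camera, triggers incremental GC and selects the first tab.
# Construct/Destruct toggle a suppressing sound mix and audio-analysis submix; loading state toggles a backdrop.
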WidgetBlueprint /ItemShop/UI/StandaloneItemShop
graphs [EventGraph {nodes [
   RegisterTab!(
    execute
    self:Fort Standalone Frontend Tabs Object Reference
    TabInfo_TabTitle:Text="LOCTEXT(\"Locker\")"
    TabInfo_TabNameId:Name=InGameLocker
    TabInfo_TabIcon:Object Reference
    TabInfo_MobileTabIcon:Slate Brush Structure=(TintColor=(SpecifiedColor=(R=1,G=1,B=1,A=1),ColorUseRule=UseColor_Specified),DrawAs=Image,Tiling=NoTile,Mirroring=NoMirror,ImageType=NoImage,ImageSize=(X=32,Y=32),Margin=(Left=0,Top=0,Right=0,Bottom=0),ResourceObject=None,OutlineSettings=(CornerRadii=(X=0,Y=0,Z=0,W=0),Color=(SpecifiedColor=(R=0,G=0,B=0,A=0),ColorUseRule=UseColor_Specified),Width=0,RoundingType=HalfHeightRadius,bUseBrushTransparency=False),UVRegion=(Min=(X=0,Y=0),Max=(X=0,Y=0),bIsValid=False),bIsDynamicallyLoaded=False,ResourceName="")
    TabInfo_TabWidget:Common Activatable Widget Object Reference=Get ActivatableContent_Locker() -> (ActivatableContent_Locker:Fort MPLocker Object Reference)
    TabInfo_ExternalTabWidget:Common Activatable Widget Soft Class Reference
    TabInfo_UIFeature:EFortUIFeature Enum=ShowLocker
    TabInfo_EnableWithCVar:String
    TabInfo_SortOrder:int=0
    TabInfo_bRequiresPocketLevel:bool=True
   ) -> (then)
   > RegisterTab(
    TabInfo_TabTitle="LOCTEXT(\"V-Bucks\")"
    TabInfo_TabNameId=InGameVBucks
    TabInfo_MobileTabIcon=(TintColor=(SpecifiedColor=(R=1,G=1,B=1,A=1),ColorUseRule=UseColor_Specified),DrawAs=Image,Tiling=NoTile,Mirroring=NoMirror,ImageType=NoImage,ImageSize=(X=32,Y=32),Margin=(Left=0,Top=0,Right=0,Bottom=0),ResourceObject=None,OutlineSettings=(CornerRadii=(X=0,Y=0,Z=0,W=0),Color=(SpecifiedColor=(R=0,G=0,B=0,A=0),ColorUseRule=UseColor_Specified),Width=0,RoundingType=HalfHeightRadius,bUseBrushTransparency=False),UVRegion=(Min=(X=0,Y=0),Max=(X=0,Y=0),bIsValid=False),bIsDynamicallyLoaded=False,ResourceName="")
    TabInfo_TabWidget=Get ActivatableContent_MtxStore() -> (ActivatableContent_MtxStore:Fort Mtx Store Root Base Object Reference)
    TabInfo_UIFeature=ShowMainStore
    TabInfo_SortOrder=0
    TabInfo_bRequiresPocketLevel=True
   )
   Event Construct()
   > PushSoundMixModifier!(execute, InSoundMixModifier:Sound Mix Object Reference=/Game/Sounds/FortSoundClassesAndModes/UI_Standalone_Shop_Suppress_MixMod) -> (then)
   n1 = Get FortAudioAnalysisSubsystem() -> (ReturnValue:Audio Analysis Subsystem Object Reference) {is=title K2Node_GetSubsystem}
   > RegisterDefaultSubmix(self:Audio Analysis Subsystem Object Reference=@n1)
   Event Destruct()
   > PopSoundMixModifier!(execute, InSoundMixModifier:Sound Mix Object Reference=&UI_Standalone_Shop_Suppress_MixMod) -> (then)
   > UnRegisterDefaultSubmix(self:Audio Analysis Subsystem Object Reference=@n1)
   n2 = Event OnTabSelected!() -> (OutputDelegate, then, TabIndex:int=0) {is=title K2Node_Event}
   > SetActiveWidgetIndex(self=Get Switcher_TabContent() -> (Switcher_TabContent:Common Widget Switcher Legacy Object Reference), Index=@n2.TabIndex)
   Event OnCameraLevelLoaded() {is=title K2Node_Event}
   > Delay(Duration=0)
   > SetFrontEndCamera!(
    execute
    self:Fort Front End Context Object Reference=Get FortFrontEndContext() -> (ReturnValue:Fort Front End Context Object Reference) {is=title K2Node_GetSubsystem}
    NewState:EFrontEndCamera Enum=Home
    TransitionParams:View Target Transition Params Structure=(BlendTime=0,BlendFunction=VTBlend_Cubic,BlendExp=2,bLockOutgoing=False)
   ) -> (then)
   > TriggerIncrementalGC!(execute) -> (then)
   > Delay(Duration=0)
   > SelectButtonAtIndex!(execute, self:Fort Standalone Frontend Tabs Object Reference, ButtonIndex:int=0) -> (then)
   n3 = Event OnLoadingIndicatorShown!() -> (OutputDelegate, then, bShown:bool) {is=title K2Node_Event}
   > SetVisibility(
    self=Get LoadingBackground() -> (LoadingBackground:Border Object Reference)
    InVisibility=Select(Option 0=Collapsed, Option 1=Visible, Index=@n3.bShown)
   )
   Event InitializeTabs() {is=title K2Node_Event}
   > RegisterTab(
    TabInfo_TabTitle="LOCTEXT(\"Item Shop\")"
    TabInfo_TabNameId=InGameMultiProductItemShop
    TabInfo_MobileTabIcon=(TintColor=(SpecifiedColor=(R=1,G=1,B=1,A=1),ColorUseRule=UseColor_Specified),DrawAs=Image,Tiling=NoTile,Mirroring=NoMirror,ImageType=NoImage,ImageSize=(X=32,Y=32),Margin=(Left=0,Top=0,Right=0,Bottom=0),ResourceObject=None,OutlineSettings=(CornerRadii=(X=0,Y=0,Z=0,W=0),Color=(SpecifiedColor=(R=0,G=0,B=0,A=0),ColorUseRule=UseColor_Specified),Width=0,RoundingType=HalfHeightRadius,bUseBrushTransparency=False),UVRegion=(Min=(X=0,Y=0),Max=(X=0,Y=0),bIsValid=False),bIsDynamicallyLoaded=False,ResourceName="")
    TabInfo_TabWidget=Get ActivatableContent_MPItemShop() -> (ActivatableContent_MPItemShop:Fort MPItem Shop Screen Object Reference)
    TabInfo_UIFeature=ShowMultiProductItemShop
    TabInfo_SortOrder=0
    TabInfo_bRequiresPocketLevel=True
   )
   > CreateTabs(self:Fort Standalone Frontend Tabs Object Reference)
   Event On Deactivated()
  ]}]
variables []
widgets {
 info {inheritedWidgetCount=6, parentClass=/Script/FortniteUI.FortStandaloneItemShop, rootWidgetClass=&Overlay, widgetCount=25}
 widgets [
  Overlay_1 {
   widgetClassPath=&Overlay
   widget=Overlay_1
   properties {navigation=&self:WidgetTree.Overlay_1.WidgetNavigation_0}
  }
  LoadingBackground {
   widgetClassPath=&Border
   bIsVariable=true
   parent=Overlay_1
   widget=LoadingBackground
   properties {background {
     bIsDynamicallyLoaded=false
     drawAs=Image
     imageType=NoImage
     mirroring=NoMirror
     resourceName=None
     resourceObject=/FortniteCoreUI/Frontend/Locker/FakeLockerBackground
     tiling=NoTile
     imageSize=<1024 512>
     margin=<ltrb 0 0 0 0>
     outlineSettings {
      bUseBrushTransparency=false
      roundingType=HalfHeightRadius
      width=0
      color {colorUseRule=UseColor_Specified, specifiedColor=<rgba 0 0 0 0>}
      cornerRadii=<0 0 0 0>
     }
     tintColor {colorUseRule=UseColor_Specified, specifiedColor=<rgba 1 1 1 1>}
     uVRegion {bIsValid=false, max=<0 0>, min=<0 0>}
    }}
   slot {class=OverlaySlot, properties {horizontalAlignment=HAlign_Fill, verticalAlignment=VAlign_Fill}}
  }
  VerticalBox_0 {
   widgetClassPath=&VerticalBox
   parent=Overlay_1
   widget=VerticalBox_0
   slot {class=OverlaySlot, properties {horizontalAlignment=HAlign_Fill, verticalAlignment=VAlign_Fill}}
  }
  Overlay_TabsNotNeededButCollapsed {
   widgetClassPath=&Overlay
   parent=VerticalBox_0
   widget=Overlay_TabsNotNeededButCollapsed
   slot {class=VerticalBoxSlot, properties {verticalAlignment=VAlign_Center}}
  }
  Image_TabBackground {
   widgetClassPath=&Image
   bIsVariable=true
   parent=Overlay_TabsNotNeededButCollapsed
   widget=Image_TabBackground
   properties {brush {
     bIsDynamicallyLoaded=false
     drawAs=Image
     imageType=NoImage
     mirroring=NoMirror
     resourceName=None
     resourceObject=null
     tiling=NoTile
     imageSize=<32 32>
     margin=<ltrb 0 0 0 0>
     outlineSettings {
      bUseBrushTransparency=false
      roundingType=HalfHeightRadius
      width=0
      color {colorUseRule=UseColor_Specified, specifiedColor=<rgba 0 0 0 0>}
      cornerRadii=<0 0 0 0>
     }
     tintColor {colorUseRule=UseColor_Specified, specifiedColor=<rgba 0 0 0 1>}
     uVRegion {bIsValid=false, max=<0 0>, min=<0 0>}
    }}
   slot {class=OverlaySlot, properties {horizontalAlignment=HAlign_Fill, verticalAlignment=VAlign_Fill}}
  }
  SafeZone_TabButtons {
   widgetClassPath=&SafeZone
   bIsVariable=true
   parent=Overlay_TabsNotNeededButCollapsed
   widget=SafeZone_TabButtons
   properties {padBottom=false, padRight=false}
   slot {class=OverlaySlot, properties {padding=<ltrb 22 0 0 0>}}
  }
  HorizontalBox_0 {widgetClassPath=&HorizontalBox, parent=SafeZone_TabButtons, widget=HorizontalBox_0, slot {class=SafeZoneSlot}}
  HBox_Tabs {
   widgetClassPath=&HorizontalBox
   bIsVariable=true
   parent=HorizontalBox_0
   widget=HBox_Tabs
   slot {class=HorizontalBoxSlot, properties {verticalAlignment=VAlign_Center}}
  }
  CommonVisibilityWidget_0 {
   widgetClassPath=&CommonVisibilityWidgetLegacy
   parent=HBox_Tabs
   widget=CommonVisibilityWidget_0
   properties {bShowForMouseAndKeyboard=false, verticalAlignment=VAlign_Center, padding=<ltrb 0 0 0 0>}
   slot {class=HorizontalBoxSlot, properties {verticalAlignment=VAlign_Center, padding=<ltrb 0 0 10 0>}}
  }
  PreviousTabAction {
   widgetClassPath=&CommonActionWidget
   bIsVariable=true
   parent=CommonVisibilityWidget_0
   widget=PreviousTabAction
   properties {inputActions [{dataTable=&InputActionDataTable, rowName=GenericLeftShoulder}]}
   slot {class=BorderSlot, properties {verticalAlignment=VAlign_Center, padding=<ltrb 0 0 0 0>}}
  }
  Border_0 {
   widgetClassPath=&Border
   parent=HBox_Tabs
   widget=Border_0
   properties {
    background {
     bIsDynamicallyLoaded=false
     drawAs=NoDrawType
     imageType=NoImage
     mirroring=NoMirror
     resourceName=None
     resourceObject=null
     tiling=NoTile
     imageSize=<32 32>
     margin=<ltrb 0.5 0 0.5 0>
     outlineSettings {
      bUseBrushTransparency=false
      roundingType=HalfHeightRadius
      width=0
      color {colorUseRule=UseColor_Specified, specifiedColor=<rgba 0 0 0 0>}
      cornerRadii=<0 0 0 0>
     }
     tintColor {colorUseRule=UseColor_Specified, specifiedColor=<rgba 1 1 1 1>}
     uVRegion {bIsValid=false, max=<0 0>, min=<0 0>}
    }
    padding=<ltrb 0 0 0 0>
   }
   slot {class=HorizontalBoxSlot, properties {horizontalAlignment=HAlign_Left, verticalAlignment=VAlign_Top}}
  }
  EntryBox_TabButtons {
   widgetClassPath=&DynamicEntryBox
   bInherited=true
   parent=Border_0
   widget=EntryBox_TabButtons
   properties {entryWidgetClass=/Game/Athena/HUD/BP_SkewedTabButton.BP_SkewedTabButton_C, numDesignerPreviewEntries=2, entrySpacing=<8 0>}
   slot {class=BorderSlot, properties {padding=<ltrb 0 0 0 0>}}
  }
  CommonVisibilityWidget {
   widgetClassPath=&CommonVisibilityWidgetLegacy
   parent=HBox_Tabs
   widget=CommonVisibilityWidget
   properties {bShowForMouseAndKeyboard=false, verticalAlignment=VAlign_Center, padding=<ltrb 0 0 0 0>}
   slot {class=HorizontalBoxSlot, properties {verticalAlignment=VAlign_Center, padding=<ltrb 10 0 0 0>}}
  }
  NextTabAction {
   widgetClassPath=&CommonActionWidget
   bIsVariable=true
   parent=CommonVisibilityWidget
   widget=NextTabAction
   properties {inputActions [{dataTable=&InputActionDataTable, rowName=GenericRightShoulder}]}
   slot {class=BorderSlot, properties {verticalAlignment=VAlign_Center, padding=<ltrb 0 0 0 0>}}
  }
  MTXButton {
   widgetClassPath=/Game/UI/Frontend/TopBar/MTXButton
   bIsVariable=true
   parent=Overlay_TabsNotNeededButCollapsed
   widget=MTXButton
   properties {
    onButtonBaseClicked=(&self:WidgetTree.MTXButton.HandleOnButtonClicked)
    onButtonBaseDoubleClicked=(&self:WidgetTree.MTXButton.HandleOnButtonDoubleClicked)
    onButtonBaseHovered=(&self:WidgetTree.MTXButton.HandleOnButtonHovered)
    onButtonBaseUnhovered=(&self:WidgetTree.MTXButton.HandleOnButtonUnhovered)
    onSelectedChangedBase=(&self:WidgetTree.MTXButton.HandleOnSelectedChanged)
   }
   slot {class=OverlaySlot, properties {horizontalAlignment=HAlign_Right, verticalAlignment=VAlign_Center, padding=<ltrb 0 0 10 0>}}
  }
  Switcher_TabContent {
   widgetClassPath=&CommonWidgetSwitcherLegacy
   bInherited=true
   bIsVariable=true
   parent=VerticalBox_0
   widget=Switcher_TabContent
   properties {transitionDuration=0}
   slot {class=VerticalBoxSlot, properties {size {sizeRule=Fill, value=1}}}
  }
  Content_LoadingIndicator {
   widgetClassPath=/Script/CommonUI.CommonLoadGuard
   bInherited=true
   bIsVariable=true
   parent=Switcher_TabContent
   widget=Content_LoadingIndicator
   properties {
    bShowLoading=true
    textStyle=&TextStyle-Base_C
    loadingBackgroundBrush {
     bIsDynamicallyLoaded=false
     drawAs=NoDrawType
     imageType=NoImage
     mirroring=NoMirror
     resourceName=None
     resourceObject=null
     tiling=NoTile
     imageSize=<1024 512>
     margin=<ltrb 0 0 0 0>
     outlineSettings {
      bUseBrushTransparency=false
      roundingType=HalfHeightRadius
      width=0
      color {colorUseRule=UseColor_Specified, specifiedColor=<rgba 0 0 0 0>}
      cornerRadii=<0 0 0 0>
     }
     tintColor {colorUseRule=UseColor_Specified, specifiedColor=<rgba 1 1 1 1>}
     uVRegion {bIsValid=false, max=<0 0>, min=<0 0>}
    }
   }
   slot {class=WidgetSwitcherSlot}
  }
  ActivatableContent_MPItemShop {
   widgetClassPath=/ItemShop/MultiProduct/WBP_MPItemShop_Screen
   bInherited=true
   bIsVariable=true
   parent=Switcher_TabContent
   widget=ActivatableContent_MPItemShop
   slot {class=WidgetSwitcherSlot}
  }
  ActivatableContent_Locker {
   widgetClassPath=/Locker/MPL/FortMPLocker
   bIsVariable=true
   parent=Switcher_TabContent
   widget=ActivatableContent_Locker
   slot {class=WidgetSwitcherSlot}
  }
  ActivatableContent_MtxStore {
   widgetClassPath=/FortniteCoreUI/Frontend/StoreMain/VBucksStoreScreen
   bInherited=true
   bIsVariable=true
   parent=Switcher_TabContent
   widget=ActivatableContent_MtxStore
   slot {class=WidgetSwitcherSlot}
  }
  SafeZone_TopRightButtons {
   widgetClassPath=&SafeZone
   bIsVariable=true
   parent=Overlay_1
   widget=SafeZone_TopRightButtons
   properties {padBottom=false, padTop=false}
   slot {class=OverlaySlot, properties {horizontalAlignment=HAlign_Right, padding=<ltrb 0 40 80 0>}}
  }
  HorizontalBox_2 {widgetClassPath=&HorizontalBox, parent=SafeZone_TopRightButtons, widget=HorizontalBox_2, slot {class=SafeZoneSlot}}
  WBP_PlayerInfo {
   widgetClassPath=/Game/UI/Frontend/ItemShop/MultiProduct/WBP_PlayerInfo
   bIsVariable=true
   parent=HorizontalBox_2
   widget=WBP_PlayerInfo
   slot {class=HorizontalBoxSlot}
  }
  VisibilityBorder_Touch {
   widgetClassPath=&CommonHardwareVisibilityBorder
   parent=HorizontalBox_2
   widget=VisibilityBorder_Touch
   properties {
    horizontalAlignment=HAlign_Right
    verticalAlignment=VAlign_Top
    padding=<ltrb 22 0 0 0>
    visibilityQuery {
     autoDescription=" ANY( Input.Touch )"
     tokenStreamVersion=0
     userDescription=if using touch
     queryTokenStream [0,1,1,1,0]
     tagDictionary [{tagName=Input.Touch}]
    }
   }
   slot {class=HorizontalBoxSlot, properties {horizontalAlignment=HAlign_Right, verticalAlignment=VAlign_Top, padding=<ltrb 0 0 30 0>}}
  }
  Button_CloseTouch {
   widgetClassPath=/FortUILibrary/Foundation/Buttons/CloseButton
   bInherited=true
   bIsVariable=true
   parent=VisibilityBorder_Touch
   widget=Button_CloseTouch
   properties {
    onButtonBaseClicked=(&self:WidgetTree.Button_CloseTouch.HandleOnButtonClicked)
    onButtonBaseDoubleClicked=(&self:WidgetTree.Button_CloseTouch.HandleOnButtonDoubleClicked)
    onButtonBaseHovered=(&self:WidgetTree.Button_CloseTouch.HandleOnButtonHovered)
    onButtonBaseUnhovered=(&self:WidgetTree.Button_CloseTouch.HandleOnButtonUnhovered)
    onSelectedChangedBase=(&self:WidgetTree.Button_CloseTouch.HandleOnSelectedChanged)
    hoveredSlateSoundOverride {resourceObject=/Game/Sounds/Fort_UI_Sounds/FrontEnd/ItemShop/Fort_UI_Shop_Menu_Hover_Cue}
    pressedSlateSoundOverride {resourceObject=/Game/Sounds/Fort_UI_Sounds/FrontEnd/ItemShop/Fort_UI_Shop_Menu_Click_Simple_A_Cue}
   }
   slot {class=BorderSlot, properties {horizontalAlignment=HAlign_Right, verticalAlignment=VAlign_Top, padding=<ltrb 22 0 0 0>}}
  }
 ]
}
